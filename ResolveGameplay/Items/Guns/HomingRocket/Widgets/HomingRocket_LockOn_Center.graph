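# Center lock-on crosshair widget for the homing rocket, with three markers and a striped square frame.
# Starts collapsed and listens for HomingStateMessage on the playspace event router, showing or hiding for the owning player's active flag.
# The first activation plays an intro fade, and lock-on state drives the locked-on animation forward once or in reverse.
# That animation's finish event starts a looping stripe pulse, stopped on deactivation.
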WidgetBlueprint /ResolveGameplay/Items/Guns/HomingRocket/Widgets/HomingRocket_LockOn_Center
animations [
 Anim_LockedOnComplete {
  binding_count=5
  length_seconds=0.45
  object=&self:Anim_LockedOnComplete
  sequence {
   bindings [
    AnimBoundSquare {
     id=#1
     possessed_class=Border
     tracks [
      {
       type=MovieSceneFloatTrack
       display_name=Render Opacity
       sections [{
         type=MovieSceneFloatSection
         channels [None {
           type=MovieSceneScriptingFloatChannel
           keys [[0,0],[8,0],[9,1],[11,1],[11.862,0],[14,0],[15,1],[17,1],[18,0],[20,0],[21,1],[23,1],[24,0],[26,0],[27,1]]
          }]
        }]
      }
      {
       type=MovieSceneDoubleVectorTrack
       display_name=Desired Size Scale
       sections [{
         type=MovieSceneDoubleVectorSection
         channels [
          Vector.X {type=MovieSceneScriptingDoubleChannel, keys [[11,0.5],[14,0.9]]}
          Vector.Y {type=MovieSceneScriptingDoubleChannel, keys [[11,0.5],[14,0.9]]}
         ]
        }]
      }
     ]
    }
    MarkerBottom {
     id=#2
     possessed_class=Image
     tracks [{
       type=MovieSceneWidgetMaterialTrack
       sections [{
         type=MovieSceneParameterSection
         channels [CompletedState {type=MovieSceneScriptingFloatChannel, keys [[0,0],[3,1],[6,0],[9,1],[12,0],[15,1]]}]
        }]
      }]
    }
    MarkerRight {
     id=#3
     possessed_class=Image
     tracks [{
       type=MovieSceneWidgetMaterialTrack
       sections [{type=MovieSceneParameterSection, channels [CompletedState {type=MovieSceneScriptingFloatChannel, keys [[0,0],[6,0],[9,1],[12,0],[15,1]]}]}]
      }]
    }
    AnimBound_CrossHairBorder {
     id=#4
     possessed_class=Border
     tracks [{
       type=MovieSceneDoubleVectorTrack
       display_name=Desired Size Scale
       sections [{
         type=MovieSceneDoubleVectorSection
         channels [
          Vector.X {type=MovieSceneScriptingDoubleChannel, keys [[4,1],[8,2.5]]}
          Vector.Y {type=MovieSceneScriptingDoubleChannel, keys [[4,1],[8,2.5]]}
         ]
        }]
      }]
    }
    MarkerLeft {
     id=#5
     possessed_class=Image
     tracks [{
       type=MovieSceneWidgetMaterialTrack
       sections [{type=MovieSceneParameterSection, channels [CompletedState {type=MovieSceneScriptingFloatChannel, keys [[0,0],[6,0],[9,1],[12,0],[15,1]]}]}]
      }]
    }
   ]
   display_rate {numerator=60}
  }
 }
 Anim_LockeOnComplete_Pulse {
  binding_count=3
  length_seconds=1
  object=&self:Anim_LockeOnComplete_Pulse
  sequence {bindings [
    Image_482 {
     id=#6
     possessed_class=Image
     tracks [{
       type=MovieSceneWidgetMaterialTrack
       sections [{type=MovieSceneParameterSection, channels [StripesGapAlpha {type=MovieSceneScriptingFloatChannel, keys [[0,0.25],[10.0003,0],[20,0.25]]}]}]
      }]
    }
    Image_5 {
     id=#7
     possessed_class=Image
     tracks [{
       type=MovieSceneWidgetMaterialTrack
       sections [{type=MovieSceneParameterSection, channels [StripesGapAlpha {type=MovieSceneScriptingFloatChannel, keys [[0,0.25],[10.0003,0],[20,0.25]]}]}]
      }]
    }
    Image_6 {
     id=#8
     possessed_class=Image
     tracks [{
       type=MovieSceneWidgetMaterialTrack
       sections [{type=MovieSceneParameterSection, channels [StripesGapAlpha {type=MovieSceneScriptingFloatChannel, keys [[0,0.25],[10.0003,0],[20,0.25]]}]}]
      }]
    }
   ]}
 }
 Anim_Intro {
  binding_count=1
  length_seconds=0.75
  object=&self:Anim_Intro
  sequence {bindings [SizeBox_113 {
     id=#9
     possessed_class=SizeBox
     tracks [{
       type=MovieSceneFloatTrack
       display_name=Render Opacity
       sections [{type=MovieSceneFloatSection, end_frame=15, start_frame=0, channels [None {type=MovieSceneScriptingFloatChannel, keys [[0,0],[4,0],[15,1]]}]}]
      }]
    }]}
 }
]
graphs [EventGraph {nodes [
   n1 = UpdateCrosshairVisuals!() -> (OutputDelegate, then, NewState:ENUM_HomingRocket_LockOnStates Enum=NewEnumerator0) {is=title K2Node_CustomEvent}
   > n2 = Switch on ENUM_HomingRocket_LockOnStates!(execute, Selection:ENUM_HomingRocket_LockOnStates Enum=@n1.NewState) -> (NewEnumerator0:Exec, NewEnumerator1:Exec, NewEnumerator2:Exec, NewEnumerator3:Exec, NewEnumerator5:Exec) {is=title K2Node_SwitchEnum}
   Set LockedOnCompleteAnimHasPlayed(execute=[@n2.NewEnumerator0, @n2.NewEnumerator1], LockedOnCompleteAnimHasPlayed:bool=false) -> (Output_Get:bool)
   n3 = Get Anim_LockedOnComplete() -> (Anim_LockedOnComplete:Widget Animation Object Reference)
   > PlayAnimation(InAnimation=@n3, StartAtTime=GetEndTime(self=@n3), PlayMode=Reverse)
   n4 = Branch(execute=@n2.NewEnumerator2, Condition=Get LockedOnCompleteAnimHasPlayed() -> (LockedOnCompleteAnimHasPlayed:bool))
   PlayAnimation(execute=@n4.else, InAnimation=Get Anim_LockedOnComplete())
   > n5 = Set LockedOnCompleteAnimHasPlayed(LockedOnCompleteAnimHasPlayed=true)
   n6 = TriggerPulseAnim() {is=title K2Node_CustomEvent}
   BindToAnimationFinished(execute=@n5.then, Animation=Get Anim_LockedOnComplete(), Delegate=@n6)
   Event Construct()
   > SetVisibility(InVisibility=Collapsed)
   > n7 = GetEventRouter!(
    execute
    Scope:Object Class Reference=/Script/FortniteGame.FortPlayspace
    ContextActor:Actor Object Reference=GetOwningPlayerPawn!(self:User Widget Object Reference) -> (ReturnValue:Pawn Object Reference)
   ) -> (then, ReturnValue:Gameplay Event Router Component Object Reference)
   > n8 = StartListeningToEvent!(
    execute:Exec
    Target:Gameplay Event Router Component Object Reference=@n7
    EventType:Script Struct Object Reference=&/HomingStateMessage
    Context:Object Reference
    EventBubblingRule:EEventBubblingRule Enum
    None:%=@n7
   ) -> (
    then:Exec
    AsyncTaskProxy:Async Action Start Listening To Event Object Reference
    OnEventReceived:Exec
    Payload:Homing State Message Structure
   )
   n9 = Break Homing State Message!(HomingStateMessage:"Homing State Message Structure (by ref)"=@n8.Payload) -> (
    LockOn_State_2_#10:ENUM_HomingRocket_LockOnStates Enum
    bIsActive_4_#11:bool=~
    MagazineAmmoCount_7_#12:int
    Player_10_#13:Fort Player Pawn Object Reference
   ) {is=title K2Node_BreakStruct}
   Branch(execute=@n8.OnEventReceived, Condition="Equal (Object)"(A=@n9.Player_10_#13, B=GetOwningPlayerPawn()))
   > n10 = Branch(Condition=@n9.bIsActive_4_#11)
   > SetVisibility(InVisibility=SelfHitTestInvisible)
   > n11 = Branch(Condition=Get Intro Anim Has Played() -> (Intro Anim Has Played:bool))
   UpdateCrosshairVisuals!(execute=[@n12.then, @n11.then], self, NewState:ENUM_HomingRocket_LockOnStates Enum=@n9) -> (then) {is=title K2Node_CallFunction}
   PlayAnimationForward(execute=@n11.else, InAnimation=Get Anim_Intro() -> (Anim_Intro:Widget Animation Object Reference))
   > n12 = Set Intro Anim Has Played(Intro Anim Has Played:bool=true) -> (Output_Get:bool)
   SetVisibility(execute=@n10.else, InVisibility=Collapsed)
   > Set Intro Anim Has Played(Intro Anim Has Played=false)
   StopAnimation(InAnimation=Get Anim_LockeOnComplete_Pulse() -> (Anim_LockeOnComplete_Pulse:Widget Animation Object Reference))
   PlayAnimation(execute=@n6.then, InAnimation=Get Anim_LockeOnComplete_Pulse(), NumLoopsToPlay=0, PlaybackSpeed=2)
  ]}]
variables [
 Intro Anim Has Played {type=boolean}
 LockedOnCompleteAnimHasPlayed {type=boolean}
]
widgets {
 info {inheritedWidgetCount=0, parentClass=&UserWidget, rootWidgetClass=&SizeBox, widgetCount=23}
 widgets [
  SizeBox_113 {
   widgetClassPath=&SizeBox
   widget=SizeBox_113
   properties {bOverride_HeightOverride=true, bOverride_WidthOverride=true, heightOverride=384, widthOverride=384}
  }
  Overlay_100 {
   widgetClassPath=&Overlay
   parent=SizeBox_113
   widget=Overlay_100
   slot {class=SizeBoxSlot, properties {horizontalAlignment=HAlign_Center, verticalAlignment=VAlign_Center}}
  }
  AnimBound_CrossHairBorder {
   widgetClassPath=&Border
   parent=Overlay_100
   widget=AnimBound_CrossHairBorder
   properties {
    background {
     bIsDynamicallyLoaded=false
     drawAs=NoDrawType
     imageType=NoImage
     mirroring=NoMirror
     resourceName=None
     resourceObject=null
     tiling=NoTile
     imageSize=<32 32>
     margin=<ltrb 0 0 0 0>
     outlineSettings {
      bUseBrushTransparency=false
      roundingType=HalfHeightRadius
      width=0
      color {colorUseRule=UseColor_Specified, specifiedColor=<rgba 0 0 0 0>}
      cornerRadii=<0 0 0 0>
     }
     tintColor {colorUseRule=UseColor_Specified, specifiedColor=<rgba 1 1 1 1>}
     uVRegion {bIsValid=false, max=<0 0>, min=<0 0>}
    }
    padding=<ltrb 0 0 0 0>
   }
   slot {class=OverlaySlot, properties {horizontalAlignment=HAlign_Fill, verticalAlignment=VAlign_Fill}}
  }
  CrossHairs {
   widgetClassPath=&SizeBox
   parent=AnimBound_CrossHairBorder
   widget=CrossHairs
   properties {bOverride_HeightOverride=true, bOverride_WidthOverride=true, heightOverride=128, widthOverride=128}
   slot {class=BorderSlot, properties {padding=<ltrb 0 0 0 0>}}
  }
  Overlay_196 {widgetClassPath=&Overlay, parent=CrossHairs, widget=Overlay_196, slot {class=SizeBoxSlot}}
  Crosshair_MarkerLeft {
   widgetClassPath=&SizeBox
   parent=Overlay_196
   widget=Crosshair_MarkerLeft
   properties {bOverride_HeightOverride=true, bOverride_WidthOverride=true, heightOverride=12, widthOverride=48}
   slot {class=OverlaySlot, properties {verticalAlignment=VAlign_Center}}
  }
  MarkerLeft {
   widgetClassPath=&Image
   bIsVariable=true
   parent=Crosshair_MarkerLeft
   widget=MarkerLeft
   properties {brush {
     bIsDynamicallyLoaded=false
     drawAs=Box
     imageType=NoImage
     mirroring=NoMirror
     resourceName=None
     resourceObject=/ResolveGameplay/Items/Guns/HomingRocket/Widgets/UI_Materials/M_UI_HomingRocket_Crosshair_Marker
     tiling=NoTile
     imageSize=<12 12>
     margin=<ltrb 0.5 0.5 0.5 0.5>
     outlineSettings {
      bUseBrushTransparency=false
      roundingType=HalfHeightRadius
      width=0
      color {colorUseRule=UseColor_Specified, specifiedColor=<rgba 0 0 0 0>}
      cornerRadii=<0 0 0 0>
     }
     tintColor {colorUseRule=UseColor_Specified, specifiedColor=<rgba 1 1 1 1>}
     uVRegion {bIsValid=false, max=<0 0>, min=<0 0>}
    }}
   slot {class=SizeBoxSlot}
  }
  Crosshair_MarkerRight {
   widgetClassPath=&SizeBox
   parent=Overlay_196
   widget=Crosshair_MarkerRight
   properties {bOverride_HeightOverride=true, bOverride_WidthOverride=true, heightOverride=12, widthOverride=48}
   slot {class=OverlaySlot, properties {horizontalAlignment=HAlign_Right, verticalAlignment=VAlign_Center}}
  }
  MarkerRight {
   widgetClassPath=&Image
   bIsVariable=true
   parent=Crosshair_MarkerRight
   widget=MarkerRight
   properties {brush {
     bIsDynamicallyLoaded=false
     drawAs=Box
     imageType=NoImage
     mirroring=NoMirror
     resourceName=None
     resourceObject=&M_UI_HomingRocket_Crosshair_Marker
     tiling=NoTile
     imageSize=<12 12>
     margin=<ltrb 0.5 0.5 0.5 0.5>
     outlineSettings {
      bUseBrushTransparency=false
      roundingType=HalfHeightRadius
      width=0
      color {colorUseRule=UseColor_Specified, specifiedColor=<rgba 0 0 0 0>}
      cornerRadii=<0 0 0 0>
     }
     tintColor {colorUseRule=UseColor_Specified, specifiedColor=<rgba 1 1 1 1>}
     uVRegion {bIsValid=false, max=<0 0>, min=<0 0>}
    }}
   slot {class=SizeBoxSlot}
  }
  Crosshair_MarkerBottm {
   widgetClassPath=&SizeBox
   parent=Overlay_196
   widget=Crosshair_MarkerBottm
   properties {bOverride_HeightOverride=true, bOverride_WidthOverride=true, heightOverride=48, widthOverride=12}
   slot {class=OverlaySlot, properties {horizontalAlignment=HAlign_Center, verticalAlignment=VAlign_Bottom}}
  }
  MarkerBottom {
   widgetClassPath=&Image
   bIsVariable=true
   parent=Crosshair_MarkerBottm
   widget=MarkerBottom
   properties {brush {
     bIsDynamicallyLoaded=false
     drawAs=Box
     imageType=NoImage
     mirroring=NoMirror
     resourceName=None
     resourceObject=&M_UI_HomingRocket_Crosshair_Marker
     tiling=NoTile
     imageSize=<12 12>
     margin=<ltrb 0.5 0.5 0.5 0.5>
     outlineSettings {
      bUseBrushTransparency=false
      roundingType=HalfHeightRadius
      width=0
      color {colorUseRule=UseColor_Specified, specifiedColor=<rgba 0 0 0 0>}
      cornerRadii=<0 0 0 0>
     }
     tintColor {colorUseRule=UseColor_Specified, specifiedColor=<rgba 1 1 1 1>}
     uVRegion {bIsValid=false, max=<0 0>, min=<0 0>}
    }}
   slot {class=SizeBoxSlot}
  }
  AnimBoundSquare {
   widgetClassPath=&Border
   parent=Overlay_100
   widget=AnimBoundSquare
   properties {
    renderOpacity=0
    visibility=SelfHitTestInvisible
    background {
     bIsDynamicallyLoaded=false
     drawAs=NoDrawType
     imageType=NoImage
     mirroring=NoMirror
     resourceName=None
     resourceObject=null
     tiling=NoTile
     imageSize=<32 32>
     margin=<ltrb 0 0 0 0>
     outlineSettings {
      bUseBrushTransparency=false
      roundingType=HalfHeightRadius
      width=0
      color {colorUseRule=UseColor_Specified, specifiedColor=<rgba 0 0 0 0>}
      cornerRadii=<0 0 0 0>
     }
     tintColor {colorUseRule=UseColor_Specified, specifiedColor=<rgba 1 1 1 1>}
     uVRegion {bIsValid=false, max=<0 0>, min=<0 0>}
    }
    desiredSizeScale=<0.5 0.5>
    padding=<ltrb 0 0 0 0>
   }
   slot {class=OverlaySlot, properties {horizontalAlignment=HAlign_Center, verticalAlignment=VAlign_Center}}
  }
  Border {widgetClassPath=&Overlay, parent=AnimBoundSquare, widget=Border, slot {class=BorderSlot, properties {padding=<ltrb 0 0 0 0>}}}
  TopLeft {widgetClassPath=&Overlay, parent=Border, widget=TopLeft, slot {class=OverlaySlot}}
  SizeBox_1 {
   widgetClassPath=&SizeBox
   parent=TopLeft
   widget=SizeBox_1
   properties {bOverride_HeightOverride=true, bOverride_WidthOverride=true, heightOverride=256, widthOverride=10}
   slot {class=OverlaySlot}
  }
  Image_482 {
   widgetClassPath=&Image
   parent=SizeBox_1
   widget=Image_482
   properties {brush {
     bIsDynamicallyLoaded=false
     drawAs=Box
     imageType=NoImage
     mirroring=NoMirror
     resourceName=None
     resourceObject=/ResolveGameplay/Items/Guns/HomingRocket/Widgets/UI_Materials/M_UI_HomingRocket_Crosshair_Square
     tiling=NoTile
     imageSize=<8 8>
     margin=<ltrb 0.5 0.5 0.5 0.5>
     outlineSettings {
      bUseBrushTransparency=false
      roundingType=HalfHeightRadius
      width=0
      color {colorUseRule=UseColor_Specified, specifiedColor=<rgba 0 0 0 0>}
      cornerRadii=<0 0 0 0>
     }
     tintColor {colorUseRule=UseColor_Specified, specifiedColor=<rgba 1 1 1 1>}
     uVRegion {bIsValid=false, max=<0 0>, min=<0 0>}
    }}
   slot {class=SizeBoxSlot}
  }
  SizeBox {
   widgetClassPath=&SizeBox
   parent=TopLeft
   widget=SizeBox
   properties {bOverride_HeightOverride=true, bOverride_WidthOverride=true, heightOverride=10, widthOverride=256}
   slot {class=OverlaySlot}
  }
  Image {
   widgetClassPath=&Image
   parent=SizeBox
   widget=Image
   properties {brush {
     bIsDynamicallyLoaded=false
     drawAs=Box
     imageType=NoImage
     mirroring=NoMirror
     resourceName=None
     resourceObject=&M_UI_HomingRocket_Crosshair_Square
     tiling=NoTile
     imageSize=<8 8>
     margin=<ltrb 0.5 0.5 0.5 0.5>
     outlineSettings {
      bUseBrushTransparency=false
      roundingType=HalfHeightRadius
      width=0
      color {colorUseRule=UseColor_Specified, specifiedColor=<rgba 0 0 0 0>}
      cornerRadii=<0 0 0 0>
     }
     tintColor {colorUseRule=UseColor_Specified, specifiedColor=<rgba 1 1 1 1>}
     uVRegion {bIsValid=false, max=<0 0>, min=<0 0>}
    }}
   slot {class=SizeBoxSlot}
  }
  BottomRight {
   widgetClassPath=&Overlay
   parent=Border
   widget=BottomRight
   properties {renderTransform {angle=180, scale=<1 1>, shear=<0 0>, translation=<0 0>}}
   slot {class=OverlaySlot, properties {horizontalAlignment=HAlign_Right, verticalAlignment=VAlign_Bottom}}
  }
  SizeBox_6 {
   widgetClassPath=&SizeBox
   parent=BottomRight
   widget=SizeBox_6
   properties {bOverride_HeightOverride=true, bOverride_WidthOverride=true, heightOverride=256, widthOverride=10}
   slot {class=OverlaySlot, properties {verticalAlignment=VAlign_Center}}
  }
  Image_5 {
   widgetClassPath=&Image
   parent=SizeBox_6
   widget=Image_5
   properties {brush {
     bIsDynamicallyLoaded=false
     drawAs=Box
     imageType=NoImage
     mirroring=NoMirror
     resourceName=None
     resourceObject=&M_UI_HomingRocket_Crosshair_Square
     tiling=NoTile
     imageSize=<8 8>
     margin=<ltrb 0.5 0.5 0.5 0.5>
     outlineSettings {
      bUseBrushTransparency=false
      roundingType=HalfHeightRadius
      width=0
      color {colorUseRule=UseColor_Specified, specifiedColor=<rgba 0 0 0 0>}
      cornerRadii=<0 0 0 0>
     }
     tintColor {colorUseRule=UseColor_Specified, specifiedColor=<rgba 1 1 1 1>}
     uVRegion {bIsValid=false, max=<0 0>, min=<0 0>}
    }}
   slot {class=SizeBoxSlot}
  }
  SizeBox_7 {
   widgetClassPath=&SizeBox
   parent=BottomRight
   widget=SizeBox_7
   properties {bOverride_HeightOverride=true, bOverride_WidthOverride=true, heightOverride=10, widthOverride=256}
   slot {class=OverlaySlot}
  }
  Image_6 {
   widgetClassPath=&Image
   parent=SizeBox_7
   widget=Image_6
   properties {brush {
     bIsDynamicallyLoaded=false
     drawAs=Box
     imageType=NoImage
     mirroring=NoMirror
     resourceName=None
     resourceObject=&M_UI_HomingRocket_Crosshair_Square
     tiling=NoTile
     imageSize=<8 8>
     margin=<ltrb 0.5 0.5 0.5 0.5>
     outlineSettings {
      bUseBrushTransparency=false
      roundingType=HalfHeightRadius
      width=0
      color {colorUseRule=UseColor_Specified, specifiedColor=<rgba 0 0 0 0>}
      cornerRadii=<0 0 0 0>
     }
     tintColor {colorUseRule=UseColor_Specified, specifiedColor=<rgba 1 1 1 1>}
     uVRegion {bIsValid=false, max=<0 0>, min=<0 0>}
    }}
   slot {class=SizeBoxSlot}
  }
 ]
}
