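# Homing Rocket lock-on circle UI widget that wraps HomingRocket_LockOn_Center and toggles its own visibility by lock-on state.
# Its graph listens for HomingStateMessage on the playspace event router, filtering for the owning player's pawn.
# On a state change, it caches the state and shows itself only if active in one specific state, otherwise collapsing.
# Construct only sets SelfHitTestInvisible; nothing shown triggers the listener setup.
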WidgetBlueprint /ResolveGameplay/Items/Guns/HomingRocket/Widgets/W_HomingRocket_LockOnCircle
graphs [EventGraph {nodes [
   n1 = GetEventRouter!(execute, Scope:Object Class Reference=/Script/FortniteGame.FortPlayspace, ContextActor:Actor Object Reference) -> (then, ReturnValue:Gameplay Event Router Component Object Reference)
   > n2 = StartListeningToEvent!(
    execute:Exec
    Target:Gameplay Event Router Component Object Reference=@n1
    EventType:Script Struct Object Reference=&/HomingStateMessage
    Context:Object Reference
    EventBubblingRule:EEventBubblingRule Enum
   ) -> (
    then:Exec
    AsyncTaskProxy:Async Action Start Listening To Event Object Reference
    OnEventReceived:Exec
    Payload:Homing State Message Structure
   )
   n3 = Break Homing State Message!(HomingStateMessage:"Homing State Message Structure (by ref)"=@n2.Payload) -> (
    LockOn_State_2_#1:ENUM_HomingRocket_LockOnStates Enum
    bIsActive_4_#2:bool=~
    MagazineAmmoCount_7_#3:int
    Player_10_#4:Fort Player Pawn Object Reference
   ) {is=title K2Node_BreakStruct}
   Branch(
    execute=@n2.OnEventReceived
    Condition="Equal (Object)"(
     A=@n3.Player_10_#4
     B=GetOwningPlayerPawn!(self:User Widget Object Reference) -> (ReturnValue:Pawn Object Reference)
    )
   )
   > Branch(Condition="Not Equal (Enum)"!(
     A:ENUM_HomingRocket_LockOnStates Enum=Get CurrentLockOnState() -> (CurrentLockOnState:ENUM_HomingRocket_LockOnStates Enum=NewEnumerator0)
     B:ENUM_HomingRocket_LockOnStates Enum=@n3
    ) -> (ReturnValue:bool))
   > Set CurrentLockOnState(CurrentLockOnState:ENUM_HomingRocket_LockOnStates Enum=@n3) -> (Output_Get:ENUM_HomingRocket_LockOnStates Enum=NewEnumerator0)
   > n4 = Branch(Condition=AND Boolean(
     A=@n3.bIsActive_4_#2
     B="Equal (Enum)"(A:ENUM_HomingRocket_LockOnStates Enum=@n3, B:ENUM_HomingRocket_LockOnStates Enum=NewEnumerator1)
    ))
   > SetVisibility(InVisibility=HitTestInvisible)
   SetVisibility(execute=@n4.else, InVisibility=Collapsed)
   Event Construct()
   > SetVisibility(InVisibility=SelfHitTestInvisible)
  ]}]
variables [CurrentLockOnState {type=ENUM_HomingRocket_LockOnStates, default=TargetingOffline}]
widgets {
 info {
  inheritedWidgetCount=0
  parentClass=&UserWidget
  rootWidgetClass=&/HomingRocket_LockOn_Center.HomingRocket_LockOn_Center_C
  widgetCount=1
 }
 widgets [HomingRocket_LockOn_Center {
   widgetClassPath=&/HomingRocket_LockOn_Center
   bIsVariable=true
   widget=HomingRocket_LockOn_Center
  }]
}
