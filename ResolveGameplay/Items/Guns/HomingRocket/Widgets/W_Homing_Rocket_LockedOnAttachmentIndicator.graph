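# Homing rocket UI widget that shows a locked-on indicator made of four glowing triangle brackets, presumably on the rocket's attached target.
# It layers two identical bracket sets, the second using material instances and starting invisible.
# PlayAttachmentLockonAnim plays a one-shot flicker-in and shrink on the first set, alongside an endlessly looping pulse that shrinks and fades the second set.
# The event graph is empty, so playback relies on that function being called.
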WidgetBlueprint /ResolveGameplay/Items/Guns/HomingRocket/Widgets/W_Homing_Rocket_LockedOnAttachmentIndicator
animations [
 Anim_LockedOnComplete_Attachment {
  binding_count=1
  length_seconds=0.25
  object=&self:Anim_LockedOnComplete_Attachment
  sequence {
   bindings [AnimBoundTrianglesBorder {
     id=#1
     possessed_class=Border
     tracks [
      {
       type=MovieSceneFloatTrack
       display_name=Render Opacity
       sections [{
         type=MovieSceneFloatSection
         channels [None {type=MovieSceneScriptingFloatChannel, thinned_from=8, keys [[0,0],[2,0],[3,0.25],[5,0.25],[6,0],[9,0],[15,1]]}]
        }]
      }
      {
       type=MovieSceneDoubleVectorTrack
       display_name=Desired Size Scale
       sections [{
         type=MovieSceneDoubleVectorSection
         channels [
          Vector.X {type=MovieSceneScriptingDoubleChannel, keys [[0,1.25],[8,1.25],[9,0.9]]}
          Vector.Y {type=MovieSceneScriptingDoubleChannel, keys [[0,1.25],[8,1.25],[9,0.9]]}
         ]
        }]
      }
     ]
    }]
   display_rate {numerator=60}
  }
 }
 Anim_LockedOnComplete_Pulse {
  binding_count=1
  length_seconds=1.0016
  object=&self:Anim_LockedOnComplete_Pulse
  sequence {bindings [AnimBoundTrianglesBorder_1 {
     id=#2
     possessed_class=Border
     tracks [
      {
       type=MovieSceneFloatTrack
       display_name=Render Opacity
       sections [{type=MovieSceneFloatSection, channels [None {type=MovieSceneScriptingFloatChannel, keys [[-0.0067,0],[4.9933,0],[5.9933,1],[11.9687,0]]}]}]
      }
      {
       type=MovieSceneDoubleVectorTrack
       display_name=Desired Size Scale
       sections [{
         type=MovieSceneDoubleVectorSection
         channels [
          Vector.X {type=MovieSceneScriptingDoubleChannel, keys [[6.0247,0.9],[11.9687,0.5],[20.0247,0.5]]}
          Vector.Y {type=MovieSceneScriptingDoubleChannel, keys [[6.0247,0.9],[11.9687,0.5],[20.0247,0.5]]}
         ]
        }]
      }
     ]
    }]}
 }
]
graphs [
 EventGraph {nodes [
   Event PreConstruct()
   Event Construct()
   Event Tick()
  ]}
 PlayAttachmentLockonAnim {nodes [
   PlayAttachmentLockonAnim() {is=title K2Node_FunctionEntry}
   > PlayAnimationForward(InAnimation=Get Anim_LockedOnComplete_Attachment() -> (Anim_LockedOnComplete_Attachment:Widget Animation Object Reference))
   > PlayAnimation(InAnimation=Get Anim_LockedOnComplete_Pulse() -> (Anim_LockedOnComplete_Pulse:Widget Animation Object Reference), NumLoopsToPlay=0)
  ]}
]
variables []
widgets {
 info {inheritedWidgetCount=0, parentClass=&UserWidget, rootWidgetClass=&SizeBox, widgetCount=24}
 widgets [
  SizeBox_0 {
   widgetClassPath=&SizeBox
   widget=SizeBox_0
   properties {bOverride_HeightOverride=true, bOverride_WidthOverride=true, heightOverride=256, widthOverride=256}
  }
  Overlay_0 {
   widgetClassPath=&Overlay
   parent=SizeBox_0
   widget=Overlay_0
   slot {class=SizeBoxSlot, properties {horizontalAlignment=HAlign_Center, verticalAlignment=VAlign_Center}}
  }
  AnimBoundTrianglesBorder {
   widgetClassPath=&Border
   bIsVariable=true
   parent=Overlay_0
   widget=AnimBoundTrianglesBorder
   properties {
    background {
     bIsDynamicallyLoaded=false
     drawAs=NoDrawType
     imageType=NoImage
     mirroring=NoMirror
     resourceName=None
     resourceObject=null
     tiling=NoTile
     imageSize=<32 32>
     margin=<ltrb 0 0 0 0>
     outlineSettings {
      bUseBrushTransparency=false
      roundingType=HalfHeightRadius
      width=0
      color {colorUseRule=UseColor_Specified, specifiedColor=<rgba 0 0 0 0>}
      cornerRadii=<0 0 0 0>
     }
     tintColor {colorUseRule=UseColor_Specified, specifiedColor=<rgba 1 1 1 1>}
     uVRegion {bIsValid=false, max=<0 0>, min=<0 0>}
    }
    padding=<ltrb 0 0 0 0>
   }
   slot {class=OverlaySlot, properties {horizontalAlignment=HAlign_Center, verticalAlignment=VAlign_Center}}
  }
  TrianglesSizeBox {
   widgetClassPath=&SizeBox
   parent=AnimBoundTrianglesBorder
   widget=TrianglesSizeBox
   properties {bOverride_HeightOverride=true, bOverride_WidthOverride=true, heightOverride=184, widthOverride=184}
   slot {class=BorderSlot, properties {padding=<ltrb 0 0 0 0>}}
  }
  TriangleBrackets {widgetClassPath=&Overlay, parent=TrianglesSizeBox, widget=TriangleBrackets, slot {class=SizeBoxSlot}}
  Triangle_TopLeft_Glow {
   widgetClassPath=&Image
   bIsVariable=true
   parent=TriangleBrackets
   widget=Triangle_TopLeft_Glow
   properties {
    brush {
     bIsDynamicallyLoaded=false
     drawAs=Image
     imageType=NoImage
     mirroring=NoMirror
     resourceName=None
     resourceObject=/ResolveGameplay/Items/Guns/HomingRocket/Widgets/UI_Materials/M_UI_Homing_TriangleBracket_Glow_Locked
     tiling=NoTile
     imageSize=<40 40>
     margin=<ltrb 0 0 0 0>
     outlineSettings {
      bUseBrushTransparency=false
      roundingType=HalfHeightRadius
      width=0
      color {colorUseRule=UseColor_Specified, specifiedColor=<rgba 0 0 0 0>}
      cornerRadii=<0 0 0 0>
     }
     tintColor {colorUseRule=UseColor_Specified, specifiedColor=<rgba 1 1 1 1>}
     uVRegion {bIsValid=false, max=<0 0>, min=<0 0>}
    }
    renderTransform {angle=0, scale=<1 1>, shear=<0 0>, translation=<1 2>}
   }
   slot {class=OverlaySlot}
  }
  Triangle_TopLeft {
   widgetClassPath=&Image
   bIsVariable=true
   parent=TriangleBrackets
   widget=Triangle_TopLeft
   properties {brush {
     bIsDynamicallyLoaded=false
     drawAs=Image
     imageType=NoImage
     mirroring=NoMirror
     resourceName=None
     resourceObject=/ResolveGameplay/Items/Guns/HomingRocket/Widgets/UI_Materials/M_UI_Homing_TriangleBracket_Locked
     tiling=NoTile
     imageSize=<40 40>
     margin=<ltrb 0 0 0 0>
     outlineSettings {
      bUseBrushTransparency=false
      roundingType=HalfHeightRadius
      width=0
      color {colorUseRule=UseColor_Specified, specifiedColor=<rgba 0 0 0 0>}
      cornerRadii=<0 0 0 0>
     }
     tintColor {colorUseRule=UseColor_Specified, specifiedColor=<rgba 1 1 1 1>}
     uVRegion {bIsValid=false, max=<0 0>, min=<0 0>}
    }}
   slot {class=OverlaySlot}
  }
  Triangle_TopRight_Glow {
   widgetClassPath=&Image
   bIsVariable=true
   parent=TriangleBrackets
   widget=Triangle_TopRight_Glow
   properties {
    brush {
     bIsDynamicallyLoaded=false
     drawAs=Image
     imageType=NoImage
     mirroring=NoMirror
     resourceName=None
     resourceObject=&M_UI_Homing_TriangleBracket_Glow_Locked
     tiling=NoTile
     imageSize=<40 40>
     margin=<ltrb 0 0 0 0>
     outlineSettings {
      bUseBrushTransparency=false
      roundingType=HalfHeightRadius
      width=0
      color {colorUseRule=UseColor_Specified, specifiedColor=<rgba 0 0 0 0>}
      cornerRadii=<0 0 0 0>
     }
     tintColor {colorUseRule=UseColor_Specified, specifiedColor=<rgba 1 1 1 1>}
     uVRegion {bIsValid=false, max=<0 0>, min=<0 0>}
    }
    renderTransform {angle=90, scale=<1 1>, shear=<0 0>, translation=<-2 2>}
   }
   slot {class=OverlaySlot, properties {horizontalAlignment=HAlign_Right}}
  }
  Triangle_TopRight {
   widgetClassPath=&Image
   bIsVariable=true
   parent=TriangleBrackets
   widget=Triangle_TopRight
   properties {
    brush {
     bIsDynamicallyLoaded=false
     drawAs=Image
     imageType=NoImage
     mirroring=NoMirror
     resourceName=None
     resourceObject=&M_UI_Homing_TriangleBracket_Locked
     tiling=NoTile
     imageSize=<40 40>
     margin=<ltrb 0 0 0 0>
     outlineSettings {
      bUseBrushTransparency=false
      roundingType=HalfHeightRadius
      width=0
      color {colorUseRule=UseColor_Specified, specifiedColor=<rgba 0 0 0 0>}
      cornerRadii=<0 0 0 0>
     }
     tintColor {colorUseRule=UseColor_Specified, specifiedColor=<rgba 1 1 1 1>}
     uVRegion {bIsValid=false, max=<0 0>, min=<0 0>}
    }
    renderTransform {angle=90, scale=<1 1>, shear=<0 0>, translation=<0 0>}
   }
   slot {class=OverlaySlot, properties {horizontalAlignment=HAlign_Right}}
  }
  Triangle_BottomLeft_Glow {
   widgetClassPath=&Image
   bIsVariable=true
   parent=TriangleBrackets
   widget=Triangle_BottomLeft_Glow
   properties {
    brush {
     bIsDynamicallyLoaded=false
     drawAs=Image
     imageType=NoImage
     mirroring=NoMirror
     resourceName=None
     resourceObject=&M_UI_Homing_TriangleBracket_Glow_Locked
     tiling=NoTile
     imageSize=<40 40>
     margin=<ltrb 0 0 0 0>
     outlineSettings {
      bUseBrushTransparency=false
      roundingType=HalfHeightRadius
      width=0
      color {colorUseRule=UseColor_Specified, specifiedColor=<rgba 0 0 0 0>}
      cornerRadii=<0 0 0 0>
     }
     tintColor {colorUseRule=UseColor_Specified, specifiedColor=<rgba 1 1 1 1>}
     uVRegion {bIsValid=false, max=<0 0>, min=<0 0>}
    }
    renderTransform {angle=-90, scale=<1 1>, shear=<0 0>, translation=<2 -1>}
   }
   slot {class=OverlaySlot, properties {verticalAlignment=VAlign_Bottom}}
  }
  Triangle_BottomLeft {
   widgetClassPath=&Image
   bIsVariable=true
   parent=TriangleBrackets
   widget=Triangle_BottomLeft
   properties {
    brush {
     bIsDynamicallyLoaded=false
     drawAs=Image
     imageType=NoImage
     mirroring=NoMirror
     resourceName=None
     resourceObject=&M_UI_Homing_TriangleBracket_Locked
     tiling=NoTile
     imageSize=<40 40>
     margin=<ltrb 0 0 0 0>
     outlineSettings {
      bUseBrushTransparency=false
      roundingType=HalfHeightRadius
      width=0
      color {colorUseRule=UseColor_Specified, specifiedColor=<rgba 0 0 0 0>}
      cornerRadii=<0 0 0 0>
     }
     tintColor {colorUseRule=UseColor_Specified, specifiedColor=<rgba 1 1 1 1>}
     uVRegion {bIsValid=false, max=<0 0>, min=<0 0>}
    }
    renderTransform {angle=-90, scale=<1 1>, shear=<0 0>, translation=<0 0>}
   }
   slot {class=OverlaySlot, properties {verticalAlignment=VAlign_Bottom}}
  }
  Triangle_BottomRight_Glow {
   widgetClassPath=&Image
   bIsVariable=true
   parent=TriangleBrackets
   widget=Triangle_BottomRight_Glow
   properties {
    brush {
     bIsDynamicallyLoaded=false
     drawAs=Image
     imageType=NoImage
     mirroring=NoMirror
     resourceName=None
     resourceObject=&M_UI_Homing_TriangleBracket_Glow_Locked
     tiling=NoTile
     imageSize=<40 40>
     margin=<ltrb 0 0 0 0>
     outlineSettings {
      bUseBrushTransparency=false
      roundingType=HalfHeightRadius
      width=0
      color {colorUseRule=UseColor_Specified, specifiedColor=<rgba 0 0 0 0>}
      cornerRadii=<0 0 0 0>
     }
     tintColor {colorUseRule=UseColor_Specified, specifiedColor=<rgba 1 1 1 1>}
     uVRegion {bIsValid=false, max=<0 0>, min=<0 0>}
    }
    renderTransform {angle=180, scale=<1 1>, shear=<0 0>, translation=<-1 -2>}
   }
   slot {class=OverlaySlot, properties {horizontalAlignment=HAlign_Right, verticalAlignment=VAlign_Bottom}}
  }
  Triangle_BottomRight {
   widgetClassPath=&Image
   bIsVariable=true
   parent=TriangleBrackets
   widget=Triangle_BottomRight
   properties {
    brush {
     bIsDynamicallyLoaded=false
     drawAs=Image
     imageType=NoImage
     mirroring=NoMirror
     resourceName=None
     resourceObject=&M_UI_Homing_TriangleBracket_Locked
     tiling=NoTile
     imageSize=<40 40>
     margin=<ltrb 0 0 0 0>
     outlineSettings {
      bUseBrushTransparency=false
      roundingType=HalfHeightRadius
      width=0
      color {colorUseRule=UseColor_Specified, specifiedColor=<rgba 0 0 0 0>}
      cornerRadii=<0 0 0 0>
     }
     tintColor {colorUseRule=UseColor_Specified, specifiedColor=<rgba 1 1 1 1>}
     uVRegion {bIsValid=false, max=<0 0>, min=<0 0>}
    }
    renderTransform {angle=180, scale=<1 1>, shear=<0 0>, translation=<0 0>}
   }
   slot {class=OverlaySlot, properties {horizontalAlignment=HAlign_Right, verticalAlignment=VAlign_Bottom}}
  }
  AnimBoundTrianglesBorder_1 {
   widgetClassPath=&Border
   bIsVariable=true
   parent=Overlay_0
   widget=AnimBoundTrianglesBorder_1
   properties {
    renderOpacity=0
    background {
     bIsDynamicallyLoaded=false
     drawAs=NoDrawType
     imageType=NoImage
     mirroring=NoMirror
     resourceName=None
     resourceObject=null
     tiling=NoTile
     imageSize=<32 32>
     margin=<ltrb 0 0 0 0>
     outlineSettings {
      bUseBrushTransparency=false
      roundingType=HalfHeightRadius
      width=0
      color {colorUseRule=UseColor_Specified, specifiedColor=<rgba 0 0 0 0>}
      cornerRadii=<0 0 0 0>
     }
     tintColor {colorUseRule=UseColor_Specified, specifiedColor=<rgba 1 1 1 1>}
     uVRegion {bIsValid=false, max=<0 0>, min=<0 0>}
    }
    padding=<ltrb 0 0 0 0>
   }
   slot {class=OverlaySlot, properties {horizontalAlignment=HAlign_Center, verticalAlignment=VAlign_Center}}
  }
  TrianglesSizeBox_1 {
   widgetClassPath=&SizeBox
   parent=AnimBoundTrianglesBorder_1
   widget=TrianglesSizeBox_1
   properties {bOverride_HeightOverride=true, bOverride_WidthOverride=true, heightOverride=184, widthOverride=184}
   slot {class=BorderSlot, properties {padding=<ltrb 0 0 0 0>}}
  }
  TriangleBrackets_1 {widgetClassPath=&Overlay, parent=TrianglesSizeBox_1, widget=TriangleBrackets_1, slot {class=SizeBoxSlot}}
  Triangle_TopLeft_Glow_1 {
   widgetClassPath=&Image
   bIsVariable=true
   parent=TriangleBrackets_1
   widget=Triangle_TopLeft_Glow_1
   properties {
    brush {
     bIsDynamicallyLoaded=false
     drawAs=Image
     imageType=NoImage
     mirroring=NoMirror
     resourceName=None
     resourceObject=/ResolveGameplay/Items/Guns/HomingRocket/Widgets/UI_Materials/M_UI_Homing_TriangleBracket_Glow_L_inst
     tiling=NoTile
     imageSize=<40 40>
     margin=<ltrb 0 0 0 0>
     outlineSettings {
      bUseBrushTransparency=false
      roundingType=HalfHeightRadius
      width=0
      color {colorUseRule=UseColor_Specified, specifiedColor=<rgba 0 0 0 0>}
      cornerRadii=<0 0 0 0>
     }
     tintColor {colorUseRule=UseColor_Specified, specifiedColor=<rgba 1 1 1 1>}
     uVRegion {bIsValid=false, max=<0 0>, min=<0 0>}
    }
    renderTransform {angle=0, scale=<1 1>, shear=<0 0>, translation=<1 2>}
   }
   slot {class=OverlaySlot}
  }
  Triangle_TopLeft_1 {
   widgetClassPath=&Image
   bIsVariable=true
   parent=TriangleBrackets_1
   widget=Triangle_TopLeft_1
   properties {brush {
     bIsDynamicallyLoaded=false
     drawAs=Image
     imageType=NoImage
     mirroring=NoMirror
     resourceName=None
     resourceObject=/ResolveGameplay/Items/Guns/HomingRocket/Widgets/UI_Materials/M_UI_Homing_TriangleBracket_Locked_Inst
     tiling=NoTile
     imageSize=<40 40>
     margin=<ltrb 0 0 0 0>
     outlineSettings {
      bUseBrushTransparency=false
      roundingType=HalfHeightRadius
      width=0
      color {colorUseRule=UseColor_Specified, specifiedColor=<rgba 0 0 0 0>}
      cornerRadii=<0 0 0 0>
     }
     tintColor {colorUseRule=UseColor_Specified, specifiedColor=<rgba 1 1 1 1>}
     uVRegion {bIsValid=false, max=<0 0>, min=<0 0>}
    }}
   slot {class=OverlaySlot}
  }
  Triangle_TopRight_Glow_1 {
   widgetClassPath=&Image
   bIsVariable=true
   parent=TriangleBrackets_1
   widget=Triangle_TopRight_Glow_1
   properties {
    brush {
     bIsDynamicallyLoaded=false
     drawAs=Image
     imageType=NoImage
     mirroring=NoMirror
     resourceName=None
     resourceObject=&M_UI_Homing_TriangleBracket_Glow_L_inst
     tiling=NoTile
     imageSize=<40 40>
     margin=<ltrb 0 0 0 0>
     outlineSettings {
      bUseBrushTransparency=false
      roundingType=HalfHeightRadius
      width=0
      color {colorUseRule=UseColor_Specified, specifiedColor=<rgba 0 0 0 0>}
      cornerRadii=<0 0 0 0>
     }
     tintColor {colorUseRule=UseColor_Specified, specifiedColor=<rgba 1 1 1 1>}
     uVRegion {bIsValid=false, max=<0 0>, min=<0 0>}
    }
    renderTransform {angle=90, scale=<1 1>, shear=<0 0>, translation=<-2 2>}
   }
   slot {class=OverlaySlot, properties {horizontalAlignment=HAlign_Right}}
  }
  Triangle_TopRight_1 {
   widgetClassPath=&Image
   bIsVariable=true
   parent=TriangleBrackets_1
   widget=Triangle_TopRight_1
   properties {
    brush {
     bIsDynamicallyLoaded=false
     drawAs=Image
     imageType=NoImage
     mirroring=NoMirror
     resourceName=None
     resourceObject=&M_UI_Homing_TriangleBracket_Locked_Inst
     tiling=NoTile
     imageSize=<40 40>
     margin=<ltrb 0 0 0 0>
     outlineSettings {
      bUseBrushTransparency=false
      roundingType=HalfHeightRadius
      width=0
      color {colorUseRule=UseColor_Specified, specifiedColor=<rgba 0 0 0 0>}
      cornerRadii=<0 0 0 0>
     }
     tintColor {colorUseRule=UseColor_Specified, specifiedColor=<rgba 1 1 1 1>}
     uVRegion {bIsValid=false, max=<0 0>, min=<0 0>}
    }
    renderTransform {angle=90, scale=<1 1>, shear=<0 0>, translation=<0 0>}
   }
   slot {class=OverlaySlot, properties {horizontalAlignment=HAlign_Right}}
  }
  Triangle_BottomLeft_Glow_1 {
   widgetClassPath=&Image
   bIsVariable=true
   parent=TriangleBrackets_1
   widget=Triangle_BottomLeft_Glow_1
   properties {
    brush {
     bIsDynamicallyLoaded=false
     drawAs=Image
     imageType=NoImage
     mirroring=NoMirror
     resourceName=None
     resourceObject=&M_UI_Homing_TriangleBracket_Glow_L_inst
     tiling=NoTile
     imageSize=<40 40>
     margin=<ltrb 0 0 0 0>
     outlineSettings {
      bUseBrushTransparency=false
      roundingType=HalfHeightRadius
      width=0
      color {colorUseRule=UseColor_Specified, specifiedColor=<rgba 0 0 0 0>}
      cornerRadii=<0 0 0 0>
     }
     tintColor {colorUseRule=UseColor_Specified, specifiedColor=<rgba 1 1 1 1>}
     uVRegion {bIsValid=false, max=<0 0>, min=<0 0>}
    }
    renderTransform {angle=-90, scale=<1 1>, shear=<0 0>, translation=<2 -1>}
   }
   slot {class=OverlaySlot, properties {verticalAlignment=VAlign_Bottom}}
  }
  Triangle_BottomLeft_1 {
   widgetClassPath=&Image
   bIsVariable=true
   parent=TriangleBrackets_1
   widget=Triangle_BottomLeft_1
   properties {
    brush {
     bIsDynamicallyLoaded=false
     drawAs=Image
     imageType=NoImage
     mirroring=NoMirror
     resourceName=None
     resourceObject=&M_UI_Homing_TriangleBracket_Locked_Inst
     tiling=NoTile
     imageSize=<40 40>
     margin=<ltrb 0 0 0 0>
     outlineSettings {
      bUseBrushTransparency=false
      roundingType=HalfHeightRadius
      width=0
      color {colorUseRule=UseColor_Specified, specifiedColor=<rgba 0 0 0 0>}
      cornerRadii=<0 0 0 0>
     }
     tintColor {colorUseRule=UseColor_Specified, specifiedColor=<rgba 1 1 1 1>}
     uVRegion {bIsValid=false, max=<0 0>, min=<0 0>}
    }
    renderTransform {angle=-90, scale=<1 1>, shear=<0 0>, translation=<0 0>}
   }
   slot {class=OverlaySlot, properties {verticalAlignment=VAlign_Bottom}}
  }
  Triangle_BottomRight_Glow_1 {
   widgetClassPath=&Image
   bIsVariable=true
   parent=TriangleBrackets_1
   widget=Triangle_BottomRight_Glow_1
   properties {
    brush {
     bIsDynamicallyLoaded=false
     drawAs=Image
     imageType=NoImage
     mirroring=NoMirror
     resourceName=None
     resourceObject=&M_UI_Homing_TriangleBracket_Glow_L_inst
     tiling=NoTile
     imageSize=<40 40>
     margin=<ltrb 0 0 0 0>
     outlineSettings {
      bUseBrushTransparency=false
      roundingType=HalfHeightRadius
      width=0
      color {colorUseRule=UseColor_Specified, specifiedColor=<rgba 0 0 0 0>}
      cornerRadii=<0 0 0 0>
     }
     tintColor {colorUseRule=UseColor_Specified, specifiedColor=<rgba 1 1 1 1>}
     uVRegion {bIsValid=false, max=<0 0>, min=<0 0>}
    }
    renderTransform {angle=180, scale=<1 1>, shear=<0 0>, translation=<-1 -2>}
   }
   slot {class=OverlaySlot, properties {horizontalAlignment=HAlign_Right, verticalAlignment=VAlign_Bottom}}
  }
  Triangle_BottomRight_1 {
   widgetClassPath=&Image
   bIsVariable=true
   parent=TriangleBrackets_1
   widget=Triangle_BottomRight_1
   properties {
    brush {
     bIsDynamicallyLoaded=false
     drawAs=Image
     imageType=NoImage
     mirroring=NoMirror
     resourceName=None
     resourceObject=&M_UI_Homing_TriangleBracket_Locked_Inst
     tiling=NoTile
     imageSize=<40 40>
     margin=<ltrb 0 0 0 0>
     outlineSettings {
      bUseBrushTransparency=false
      roundingType=HalfHeightRadius
      width=0
      color {colorUseRule=UseColor_Specified, specifiedColor=<rgba 0 0 0 0>}
      cornerRadii=<0 0 0 0>
     }
     tintColor {colorUseRule=UseColor_Specified, specifiedColor=<rgba 1 1 1 1>}
     uVRegion {bIsValid=false, max=<0 0>, min=<0 0>}
    }
    renderTransform {angle=180, scale=<1 1>, shear=<0 0>, translation=<0 0>}
   }
   slot {class=OverlaySlot, properties {horizontalAlignment=HAlign_Right, verticalAlignment=VAlign_Bottom}}
  }
 ]
}
