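# Widget that shows the homing rocket lock-on indicator over a targeted vehicle on the player's screen.
# Construct starts collapsed and listens on the playspace event router for target messages, which feed SetTarget.
# A valid target shows the widget, plays the child's lock-on animation once and starts a looping timer; an invalid one hides it and clears the timer.
# PinToTarget centers it at a viewport position derived from the vehicle mesh's bounds origin.
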WidgetBlueprint /ResolveGameplay/Items/Guns/HomingRocket/Widgets/W_HomingRocket_LockOnIndicator_V2
graphs [
 EventGraph {nodes [
   Event Construct()
   > n1 = GetEventRouter!(execute, Scope:Object Class Reference=/Script/FortniteGame.FortPlayspace, ContextActor:Actor Object Reference=GetOwningPlayer()) -> (then, ReturnValue:Gameplay Event Router Component Object Reference)
   > n2 = StartListeningToEvent!(
    execute:Exec
    Target:Gameplay Event Router Component Object Reference=@n1
    EventType:Script Struct Object Reference=&/HomingRocket_LockOn_Indicator_Target_Message
    Context:Object Reference
    EventBubblingRule:EEventBubblingRule Enum
   ) -> (
    then:Exec
    AsyncTaskProxy:Async Action Start Listening To Event Object Reference
    OnEventReceived:Exec
    Payload:Homing Rocket Lock On Indicator Target Message Structure
   )
   > SetVisibility(InVisibility=Collapsed)
   > Set LockedOnAttachmentAnimHasPlayed(LockedOnAttachmentAnimHasPlayed:bool=false) -> (Output_Get:bool)
   SetTarget!(
    execute=@n2.OnEventReceived
    self
    InTarget:Actor Object Reference=Break Homing Rocket Lock on Indicator Target Message!(HomingRocket_LockOn_Indicator_Target_Message:"Homing Rocket Lock On Indicator Target Message Structure (by ref)"=@n2.Payload) -> (TargetActor_2_#1:Actor Object Reference) {is=title K2Node_BreakStruct}
   ) -> (then)
  ]}
 PinToTarget {nodes [
   PinToTarget() {is=title K2Node_FunctionEntry}
   > n1 = Cast To FortAthenaVehicle!(execute, Object=Get Target() -> (Target:Actor Object Reference)) -> (then, CastFailed, AsFort Athena Vehicle:Fort Athena Vehicle Object Reference) {is=title K2Node_DynamicCast}
   n2 = BroadcastPlayerImpactAtLocation!(
    Player:Player Controller Object Reference=GetOwningPlayer()
    Position:"Vector (by ref)"=GetComponentBounds!(Component:Scene Component Object Reference=Get SkeletalMesh!(self:Fort Athena Vehicle Object Reference=@n1) -> (SkeletalMesh:Skeletal Mesh Component Budgeted Object Reference)) -> (Origin:Vector, BoxExtent:Vector, SphereRadius:float)
    bRelativePosition:bool=false
   ) -> (Loudness:float, Duration:float, ReturnValue:bool)
   n3 = GetDesiredSize!(self:Widget Object Reference) -> (ReturnValue:Vector 2D Structure)
   > SetPositionInViewport!(
    execute
    self:User Widget Object Reference
    Position:Vector 2D Structure=vector2d - vector2d(
     A:Vector 2D Structure=MakeVector2D(X=@n2, Y=@n2.Duration)
     B:Vector 2D Structure=vector2d * vector2d(A:Vector 2D Structure=@n3, B:Vector 2D Structure=(X=0.5,Y=0.5)) -> (ReturnValue:Vector 2D Structure) {is=title K2Node_PromotableOperator}
    ) -> (ReturnValue:Vector 2D Structure) {is=title K2Node_PromotableOperator}
    bRemoveDPIScale:bool=true
   ) -> (then)
  ]}
 SetTarget {nodes [
   n1 = SetTarget!() -> (then, InTarget:Actor Object Reference) {is=title K2Node_FunctionEntry}
   > Set Target(Target:Actor Object Reference=@n1) -> (Output_Get:Actor Object Reference)
   > n2 = Is Valid(InputObject=Get Target())
   SetVisibility(execute=@n2.Is Not Valid, InVisibility=Collapsed)
   > Clear and Invalidate Timer by Handle(Handle=Get PinToTargetTimerHandle() -> (PinToTargetTimerHandle:Timer Handle Structure))
   SetVisibility(execute=@n2.Is Valid, InVisibility=HitTestInvisible)
   > n3 = Set Timer by Event(Delegate=Create Event(), Time=0.025, bLooping=true)
   > Set PinToTargetTimerHandle(PinToTargetTimerHandle:Timer Handle Structure=@n3) -> (Output_Get:Timer Handle Structure)
   > n4 = Branch(Condition=Get LockedOnAttachmentAnimHasPlayed() -> (LockedOnAttachmentAnimHasPlayed:bool))
   PlayAttachmentLockonAnim(
    execute=@n4.else
    self:W Homing Rocket Locked On Attachment Indicator Object Reference=Get W_Homing_Rocket_LockedOnAttachmentIndicator() -> (W_Homing_Rocket_LockedOnAttachmentIndicator:W Homing Rocket Locked On Attachment Indicator Object Reference)
   )
   > Set LockedOnAttachmentAnimHasPlayed(LockedOnAttachmentAnimHasPlayed=true)
  ]}
]
variables [
 LockedOnAttachmentAnimHasPlayed {type=boolean}
 PinToTargetTimerHandle {type=TimerHandle}
 Target {type=Actor, default=null}
]
widgets {
 info {
  inheritedWidgetCount=0
  parentClass=/Script/FortniteUI.FortUserWidget
  rootWidgetClass=&/W_Homing_Rocket_LockedOnAttachmentIndicator.W_Homing_Rocket_LockedOnAttachmentIndicator_C
  widgetCount=1
 }
 widgets [W_Homing_Rocket_LockedOnAttachmentIndicator {
   widgetClassPath=&/W_Homing_Rocket_LockedOnAttachmentIndicator
   bIsVariable=true
   widget=W_Homing_Rocket_LockedOnAttachmentIndicator
  }]
}
