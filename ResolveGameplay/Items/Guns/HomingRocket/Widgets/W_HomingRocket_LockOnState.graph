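# HUD widget for the homing rocket that shows searching, locking-on, locked-on and out-of-ammo states.
# It listens on FortPlayspace event routers for state messages, filtered to the owning pawn, and for acquiring-target progress.
# State functions switch text, recolor dynamic materials and play one-shot animations, while progress drives a material parameter and percent text.
# While active it plays an intro and hides other HUD elements by tag, restoring them when inactive.
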WidgetBlueprint /ResolveGameplay/Items/Guns/HomingRocket/Widgets/W_HomingRocket_LockOnState
animations [
 Anim_IsLockingOn {
  binding_count=2
  length_seconds=0.15
  object=&self:Anim_IsLockingOn
  sequence {
   bindings [
    Progress {
     id=#1
     possessed_class=Image
     tracks [{
       type=MovieSceneWidgetMaterialTrack
       sections [{
         type=MovieSceneParameterSection
         channels [IsLockingOn {type=MovieSceneScriptingFloatChannel, keys [[0,0],[1.962,0],[3,0.5],[5,0.5],[6,0],[8,0],[9,1]]}]
        }]
      }]
    }
    Progress_Container {
     id=#2
     possessed_class=Image
     tracks [{
       type=MovieSceneWidgetMaterialTrack
       sections [{
         type=MovieSceneParameterSection
         channels [IsLockingOn {type=MovieSceneScriptingFloatChannel, keys [[0,0],[1.962,0],[3,0.5],[5,0.5],[6,0],[8,0],[9,1]]}]
        }]
      }]
    }
   ]
   display_rate {numerator=60}
  }
 }
 Anim_Intro {
  binding_count=3
  length_seconds=0.5
  object=&self:Anim_Intro
  sequence {
   bindings [
    CenterLineRight {
     id=#3
     possessed_class=SizeBox
     tracks [{
       type=MovieScene2DTransformTrack
       sections [{
         type=MovieScene2DTransformSection
         channels [
          Translation.X {type=MovieSceneScriptingFloatChannel}
          Translation.Y {type=MovieSceneScriptingFloatChannel}
          Angle {type=MovieSceneScriptingFloatChannel}
          Scale.X {type=MovieSceneScriptingFloatChannel, keys [[9,0.25],[30,1]]}
          Shear.X {type=MovieSceneScriptingFloatChannel}
          Shear.Y {type=MovieSceneScriptingFloatChannel}
         ]
        }]
      }]
    }
    CenterLineLeft {
     id=#4
     possessed_class=SizeBox
     tracks [{
       type=MovieScene2DTransformTrack
       sections [{
         type=MovieScene2DTransformSection
         channels [
          Translation.X {type=MovieSceneScriptingFloatChannel}
          Translation.Y {type=MovieSceneScriptingFloatChannel}
          Angle {type=MovieSceneScriptingFloatChannel}
          Scale.X {type=MovieSceneScriptingFloatChannel, keys [[9,0.25],[30,1]]}
          Shear.X {type=MovieSceneScriptingFloatChannel}
          Shear.Y {type=MovieSceneScriptingFloatChannel}
         ]
        }]
      }]
    }
    All {
     id=#5
     possessed_class=Border
     tracks [{
       type=MovieSceneFloatTrack
       display_name=Render Opacity
       sections [{
         type=MovieSceneFloatSection
         channels [None {type=MovieSceneScriptingFloatChannel, thinned_from=12, keys [[0,0],[11,0],[12,0.25],[14,0.25],[15,0],[17,0],[18,0.5],[23,0.5],[27,1]]}]
        }]
      }]
    }
   ]
   display_rate {numerator=60}
  }
 }
 Anim_LockingOnStarted_Triangles {
  binding_count=1
  length_seconds=0.3
  object=&self:Anim_LockingOnStarted_Triangles
  sequence {
   bindings [AnimBoundTrianglesBorder {
     id=#6
     possessed_class=Border
     tracks [
      {
       type=MovieSceneDoubleVectorTrack
       display_name=Desired Size Scale
       sections [{
         type=MovieSceneDoubleVectorSection
         channels [
          Vector.X {type=MovieSceneScriptingDoubleChannel, keys [[0,1.325],[3,1.325],[4,1.25],[11,1.25],[12,1.325],[14,1.0625]]}
          Vector.Y {type=MovieSceneScriptingDoubleChannel, keys [[0,1.325],[3,1.325],[4,1.25],[11,1.25],[12,1.325],[14,1.0625]]}
         ]
        }]
      }
      {
       type=MovieSceneFloatTrack
       display_name=Render Opacity
       sections [{
         type=MovieSceneFloatSection
         channels [None {type=MovieSceneScriptingFloatChannel, keys [[0,0],[3,0.25],[4,0],[7,0],[8,0.25],[11,0.25],[12,0],[14,0],[18,1]]}]
        }]
      }
     ]
    }]
   display_rate {numerator=60}
  }
 }
 Anim_LockedOnComplete_Triangles {
  binding_count=2
  length_seconds=0.2
  object=&self:Anim_LockedOnComplete_Triangles
  sequence {
   bindings [
    AnimBoundTrianglesBorder {
     id=#6
     possessed_class=Border
     tracks [
      {
       type=MovieSceneDoubleVectorTrack
       display_name=Desired Size Scale
       sections [{
         type=MovieSceneDoubleVectorSection
         channels [
          Vector.X {type=MovieSceneScriptingDoubleChannel, keys [[0,1],[9,0.5]]}
          Vector.Y {type=MovieSceneScriptingDoubleChannel, keys [[0,1],[9,0.5]]}
         ]
        }]
      }
      {
       type=MovieSceneFloatTrack
       display_name=Render Opacity
       sections [{type=MovieSceneFloatSection, channels [None {type=MovieSceneScriptingFloatChannel, keys [[0,1],[6,1],[9,0]]}]}]
      }
     ]
    }
    Progress {
     id=#7
     possessed_class=Image
     tracks [{
       type=MovieSceneFloatTrack
       display_name=Render Opacity
       sections [{
         type=MovieSceneFloatSection
         channels [None {type=MovieSceneScriptingFloatChannel, keys [[0,1],[2,1],[3,0],[5,0],[6,1],[8,1],[9,0],[11,0],[12,1]]}]
        }]
      }]
    }
   ]
   display_rate {numerator=60}
  }
 }
]
graphs [
 EventGraph {nodes [
   n1 = SetIgnoreInheritedScale!(execute, self:Scale Box Object Reference, bInIgnoreInheritedScale:bool=true) -> (then)
   n2 = Get Anim_Intro() -> (Anim_Intro:Widget Animation Object Reference)
   PlayAnimation(execute=[@n1.then, @n11.else], InAnimation=@n2, StartAtTime=GetEndTime(self=@n2), PlayMode=Reverse)
   n3 = UpdateLockOnText!() -> (OutputDelegate, then, NewState:ENUM_HomingRocket_LockOnStates Enum=NewEnumerator0, MagazineAmmoCount:int) {is=title K2Node_CustomEvent}
   > n4 = Branch(Condition="integer > integer"(A=@n3.MagazineAmmoCount))
   > n5 = Switch on ENUM_HomingRocket_LockOnStates!(execute, Selection:ENUM_HomingRocket_LockOnStates Enum=@n3.NewState) -> (NewEnumerator0:Exec, NewEnumerator1:Exec, NewEnumerator2:Exec, NewEnumerator3:Exec, NewEnumerator5:Exec) {is=title K2Node_SwitchEnum}
   > SetSearchingStateVisuals()
   SetLockingOnStateVisuals(execute=@n5.NewEnumerator1)
   SetLockedOnStateVisuals(execute=@n5.NewEnumerator2)
   SetOutOfAmmoStateVisuals(execute=@n4.else)
   n6 = UpdateLockOnProgress!() -> (OutputDelegate, then, Progress:double=~) {is=title K2Node_CustomEvent}
   > n7 = GetDynamicMaterial(self=Get Progress() -> (Progress:Image Object Reference))
   > SetScalarParameterValue(self=@n7, ParameterName=Progress, Value=@n6.Progress)
   > n8 = GetDynamicFontMaterial!(execute, self:Text Object Reference=Get PercentText() -> (PercentText:Common Text Object Reference)) -> (then, ReturnValue:Material Instance Dynamic Object Reference)
   > SetScalarParameterValue(self=@n8, ParameterName=Progress, Value=@n6.Progress)
   n9 = "Clamp (Integer)"(Value=Round(A=float * float(A=@n6.Progress, B=100) {is=title K2Node_PromotableOperator}) {is=title K2Node_CallFunction}, Min=0, Max=100) -> (ReturnValue=0)
   > "SetText (Text)"(self=Get PercentText(), InText=Format Text!(Format="LOCTEXT(\"{Percent}%\")", Percent:Text="To Text (Integer)"(Value=@n9)) -> (Result))
   n10 = Event PreConstruct()
   > n11 = Branch(Condition=@n10.IsDesignTime)
   Event Construct()
   > Set LockingOnTrianglesAnimHasPlayed(LockingOnTrianglesAnimHasPlayed:bool=false) -> (Output_Get:bool)
   > Set LockingOnAnimHasPlayed(LockingOnAnimHasPlayed:bool=false) -> (Output_Get:bool)
   > Set IntroAnimHasPlayed(IntroAnimHasPlayed:bool=false) -> (Output_Get:bool)
   > SetVisibility(InVisibility=Collapsed)
   > n12 = Sequence()
   n13 = GetOwningPlayerPawn!(self:User Widget Object Reference) -> (ReturnValue:Pawn Object Reference)
   > n14 = GetEventRouter!(execute, Scope:Object Class Reference=/Script/FortniteGame.FortPlayspace, ContextActor:Actor Object Reference=@n13) -> (then, ReturnValue:Gameplay Event Router Component Object Reference)
   > n15 = StartListeningToEvent!(
    execute:Exec
    Target:Gameplay Event Router Component Object Reference=@n14
    EventType:Script Struct Object Reference=&/HomingStateMessage
    Context:Object Reference
    EventBubblingRule:EEventBubblingRule Enum
   ) -> (
    then:Exec
    AsyncTaskProxy:Async Action Start Listening To Event Object Reference
    OnEventReceived:Exec
    Payload:Homing State Message Structure
   )
   n16 = Break Homing State Message!(HomingStateMessage:"Homing State Message Structure (by ref)"=@n15.Payload) -> (
    LockOn_State_2_#8:ENUM_HomingRocket_LockOnStates Enum
    bIsActive_4_#9:bool=~
    MagazineAmmoCount_7_#10:int
    Player_10_#11:Fort Player Pawn Object Reference
   ) {is=title K2Node_BreakStruct}
   Branch(execute=@n15.OnEventReceived, Condition="Equal (Object)"(A=@n16.Player_10_#11, B=GetOwningPlayerPawn()))
   > n17 = Branch(Condition=@n16.bIsActive_4_#9)
   > UpdateLockOnText!(
    execute
    self
    NewState:ENUM_HomingRocket_LockOnStates Enum=@n16
    MagazineAmmoCount:int=@n16.MagazineAmmoCount_7_#10
   ) -> (then) {is=title K2Node_CallFunction}
   > SetVisibility(InVisibility=HitTestInvisible)
   > n18 = Branch(Condition=Get IntroAnimHasPlayed() -> (IntroAnimHasPlayed:bool))
   PlayAnimationForward(execute=@n18.else, InAnimation=Get Anim_Intro())
   > Set IntroAnimHasPlayed(IntroAnimHasPlayed=true)
   > SetHUDElementVisibility!(
    execute
    self:Fort HUDContext Object Reference=Get FortHUDContext!() -> (ReturnValue:Fort HUDContext Object Reference) {is=title K2Node_GetBlueprintContextEx}
    HUDElementTags:"Gameplay Tag Container Structure (by ref)"=Get HUDElementsToHide() -> (HUDElementsToHide:Gameplay Tag Container Structure)
    bHideElements:bool=true
    ReasonName:"Name (by ref)"=None
   ) -> (then)
   SetVisibility(execute=@n17.else, InVisibility=Collapsed)
   > Set IntroAnimHasPlayed(IntroAnimHasPlayed=false)
   > SetHUDElementVisibility(self=Get FortHUDContext(), HUDElementTags=Get HUDElementsToHide(), bHideElements=false, ReasonName=None)
   n19 = GetEventRouter(execute=@n12.then_1, Scope=&FortPlayspace, ContextActor=@n13)
   > n20 = StartListeningToEvent(Target=@n19, EventType=&/HomingRocket_AcquiringTarget_Message) -> (Payload:Homing Rocket Acquiring Target Message Structure)
   UpdateLockOnProgress!(
    execute=@n20.OnEventReceived
    self
    Progress:double=Break Homing Rocket Acquiring Target Message!(HomingRocket_AcquiringTarget_Message:"Homing Rocket Acquiring Target Message Structure (by ref)"=@n20.Payload) -> (AcquiringPercent_2_#12:float=~, AcquiringPercentTotal_4_#13:float=~) {is=title K2Node_BreakStruct}
   ) -> (then) {is=title K2Node_CallFunction}
   SetIgnoreInheritedScale(bInIgnoreInheritedScale=false)
  ]}
 SetColorsLockedOnComplete {nodes [
   SetColorsLockedOnComplete() {is=title K2Node_FunctionEntry}
   > n1 = GetDynamicMaterial(self:Border Object Reference=Get StatusContainer() -> (StatusContainer:Border Object Reference))
   n2 = Get GlowLockingCompleteColor() -> (GlowLockingCompleteColor:Linear Color Structure)
   > SetVectorParameterValue(self=@n1, ParameterName=Color_Glow, Value=@n2)
   > SetVectorParameterValue(self=@n1, ParameterName=Color_Pip, Value=@n2)
   > SetScalarParameterValue(self=@n1, ParameterName=Alpha_Glow, Value=0.025)
   > n3 = GetDynamicMaterial(self:Image Object Reference=Get GlitchFX1() -> (GlitchFX1:Image Object Reference))
   > SetVectorParameterValue(self=@n3, ParameterName=ColorA, Value=Get GlitchFX_ColorA_LockedOnComplete() -> (GlitchFX_ColorA_LockedOnComplete:Linear Color Structure))
   > SetVectorParameterValue(self=@n3, ParameterName=ColorB, Value=Get GlitchFX_ColorB_LockedOnComplete() -> (GlitchFX_ColorB_LockedOnComplete:Linear Color Structure))
   > n4 = GetDynamicMaterial(self=Get GlitchFX2() -> (GlitchFX2:Image Object Reference))
   > SetVectorParameterValue(self=@n4, ParameterName=ColorA, Value=Get GlitchFX_ColorA_LockedOnComplete())
   > SetVectorParameterValue(self=@n4, ParameterName=ColorB, Value=Get GlitchFX_ColorB_LockedOnComplete())
   > n5 = For Each Loop(Array:Array of Image Object References=Make Array!(
     "[0]":Image Object Reference=Get Pip_TopLeft() -> (Pip_TopLeft:Image Object Reference)
     "[1]":Image Object Reference=Get Pip_TopRight() -> (Pip_TopRight:Image Object Reference)
     "[2]":Image Object Reference=Get Pip_BottomLeft() -> (Pip_BottomLeft:Image Object Reference)
     "[3]":Image Object Reference=Get Pip_BottomRight() -> (Pip_BottomRight:Image Object Reference)
    ) -> (Array:Array of Image Object References)) -> (Array Element:Image Object Reference)
   > n6 = GetDynamicMaterial(self=@n5)
   > SetVectorParameterValue(self=@n6, ParameterName=Color_Glow, Value=Get Pip_Color_Glow_LockedOnComplete() -> (Pip_Color_Glow_LockedOnComplete:Linear Color Structure))
   > SetVectorParameterValue(self=@n6, ParameterName=Color_Pip, Value=Get Pip_Color_Pip_LockedOnComplete() -> (Pip_Color_Pip_LockedOnComplete:Linear Color Structure))
  ]}
 SetColorsLockingOn {nodes [
   SetColorsLockingOn() {is=title K2Node_FunctionEntry}
   > n1 = GetDynamicMaterial(self:Border Object Reference=Get StatusContainer())
   n2 = Get GlowLockingColor() -> (GlowLockingColor:Linear Color Structure)
   > SetVectorParameterValue(self=@n1, ParameterName=Color_Glow, Value=@n2)
   > SetVectorParameterValue(self=@n1, ParameterName=Color_Pip, Value=@n2)
   > SetScalarParameterValue(self=@n1, ParameterName=Alpha_Glow, Value=0.025)
   > n3 = GetDynamicMaterial(self:Image Object Reference=Get GlitchFX1())
   > SetVectorParameterValue(self=@n3, ParameterName=ColorA, Value=Get GlitchFX_ColorA_LockingOn() -> (GlitchFX_ColorA_LockingOn:Linear Color Structure))
   > SetVectorParameterValue(self=@n3, ParameterName=ColorB, Value=Get GlitchFX_ColorB_LockingOn() -> (GlitchFX_ColorB_LockingOn:Linear Color Structure))
   > n4 = GetDynamicMaterial(self=Get GlitchFX2())
   > SetVectorParameterValue(self=@n4, ParameterName=ColorA, Value=Get GlitchFX_ColorA_LockingOn())
   > SetVectorParameterValue(self=@n4, ParameterName=ColorB, Value=Get GlitchFX_ColorB_LockingOn())
   > n5 = For Each Loop(Array=Make Array("[0]"=Get Pip_TopLeft(), "[1]"=Get Pip_TopRight(), "[2]"=Get Pip_BottomLeft(), "[3]"=Get Pip_BottomRight()))
   > n6 = GetDynamicMaterial(self=@n5)
   > SetVectorParameterValue(self=@n6, ParameterName=Color_Glow, Value=Get Pip_Color_Glow_LockingOn() -> (Pip_Color_Glow_LockingOn:Linear Color Structure))
   > SetVectorParameterValue(self=@n6, ParameterName=Color_Pip, Value=Get Pip_Color_Pip_LockingOn() -> (Pip_Color_Pip_LockingOn:Linear Color Structure))
  ]}
 SetColorsOutOfAmmo {nodes [
   SetColorsOutOfAmmo() {is=title K2Node_FunctionEntry}
   > n1 = GetDynamicMaterial(self:Border Object Reference=Get StatusContainer())
   n2 = Get GlowLockingCompleteColor()
   > SetVectorParameterValue(self=@n1, ParameterName=Color_Glow, Value=@n2)
   > SetVectorParameterValue(self=@n1, ParameterName=Color_Pip, Value=@n2)
   > SetScalarParameterValue(self=@n1, ParameterName=Alpha_Glow, Value=0.025)
   > n3 = GetDynamicMaterial(self:Image Object Reference=Get GlitchFX1())
   > SetVectorParameterValue(self=@n3, ParameterName=ColorA, Value=Get GlitchFX_ColorA_LockedOnComplete())
   > SetVectorParameterValue(self=@n3, ParameterName=ColorB, Value=Get GlitchFX_ColorB_LockedOnComplete())
   > n4 = GetDynamicMaterial(self=Get GlitchFX2())
   > SetVectorParameterValue(self=@n4, ParameterName=ColorA, Value=Get GlitchFX_ColorA_LockedOnComplete())
   > SetVectorParameterValue(self=@n4, ParameterName=ColorB, Value=Get GlitchFX_ColorB_LockedOnComplete())
   > n5 = For Each Loop(Array=Make Array("[0]"=Get Pip_TopLeft(), "[1]"=Get Pip_TopRight(), "[2]"=Get Pip_BottomLeft(), "[3]"=Get Pip_BottomRight()))
   > n6 = GetDynamicMaterial(self=@n5)
   > SetVectorParameterValue(self=@n6, ParameterName=Color_Glow, Value=Get Pip_Color_Glow_LockedOnComplete())
   > SetVectorParameterValue(self=@n6, ParameterName=Color_Pip, Value=Get Pip_Color_Pip_LockedOnComplete())
  ]}
 SetColorsSearching {nodes [
   SetColorsSearching() {is=title K2Node_FunctionEntry}
   > n1 = GetDynamicMaterial(self:Border Object Reference=Get StatusContainer())
   > SetVectorParameterValue(self=@n1, ParameterName=Color_Glow, Value=Get GlowSearchingColor() -> (GlowSearchingColor:Linear Color Structure))
   > SetVectorParameterValue(self=@n1, ParameterName=Color_Pip, Value=Get GlowSearchingColor())
   > SetScalarParameterValue(self=@n1, ParameterName=Alpha_Glow, Value=0.0063)
   > n2 = GetDynamicMaterial(self:Image Object Reference=Get GlitchFX1())
   > SetVectorParameterValue(self=@n2, ParameterName=ColorA, Value=Get GlitchFX_ColorA_Searching() -> (GlitchFX_ColorA_Searching:Linear Color Structure))
   > SetVectorParameterValue(self=@n2, ParameterName=ColorB, Value=Get GlitchFX_ColorB_Searching() -> (GlitchFX_ColorB_Searching:Linear Color Structure))
   > n3 = GetDynamicMaterial(self=Get GlitchFX2())
   > SetVectorParameterValue(self=@n3, ParameterName=ColorA, Value=Get GlitchFX_ColorA_Searching())
   > SetVectorParameterValue(self=@n3, ParameterName=ColorB, Value=Get GlitchFX_ColorB_Searching())
   > n4 = For Each Loop(Array=Make Array("[0]"=Get Pip_TopLeft(), "[1]"=Get Pip_TopRight(), "[2]"=Get Pip_BottomLeft(), "[3]"=Get Pip_BottomRight()))
   > n5 = GetDynamicMaterial(self=@n4)
   > SetVectorParameterValue(self=@n5, ParameterName=Color_Glow, Value=Get Pip_Color_Glow_Searching() -> (Pip_Color_Glow_Searching:Linear Color Structure))
   > SetVectorParameterValue(self=@n5, ParameterName=Color_Pip, Value=Get Pip_Color_Pip_Searching() -> (Pip_Color_Pip_Searching:Linear Color Structure))
  ]}
 SetLockedOnStateVisuals {nodes [
   n1 = "SetText (Text)"(
    self=Get PercentText()
    InText=Format Text!(Format="LOCTEXT(\"{Value}%\")", Value:int=MakeLiteralInt!(Value:int=100) -> (ReturnValue:int=0)) -> (Result)
   )
   SetActiveWidget(
    execute=[@n1.then, @n5.then]
    self:Widget Switcher Object Reference=Get CompletedLockonTextSwitcher() -> (CompletedLockonTextSwitcher:Widget Switcher Object Reference)
    Widget=Get PercentText_Completed() -> (PercentText_Completed:Common Text Object Reference)
   )
   > SetActiveWidget(
    self=Get Text!() -> (Text:Widget Switcher Object Reference)
    Widget=Get LockedOnCompleteText() -> (LockedOnCompleteText:Border Object Reference)
   )
   > SetVisibility(self=Get Percent() -> (Percent:Border Object Reference), InVisibility=SelfHitTestInvisible)
   > n2 = GetDynamicMaterial(self=Get Progress())
   > SetScalarParameterValue(self=@n2, ParameterName=Progress, Value=1)
   > SetScalarParameterValue(self=@n2, ParameterName=IsLockedOn, Value=1)
   > n3 = GetDynamicMaterial(self=Get Progress_Container() -> (Progress_Container:Image Object Reference))
   > SetScalarParameterValue(self=@n3, ParameterName=IsLockedOn, Value=1)
   > SetColorsLockedOnComplete() {is=title K2Node_CallFunction}
   > SetSquareBracketsColors!(execute, self, IsCompleted:double=0) -> (then)
   > Set LockingOnAnimHasPlayed(LockingOnAnimHasPlayed=false)
   > n4 = Branch(Condition=Get LockingOnTrianglesAnimHasPlayed() -> (LockingOnTrianglesAnimHasPlayed:bool))
   PlayAnimationForward(execute=@n4.else, InAnimation=Get Anim_LockedOnComplete_Triangles() -> (Anim_LockedOnComplete_Triangles:Widget Animation Object Reference))
   > Set LockingOnTrianglesAnimHasPlayed(LockingOnTrianglesAnimHasPlayed=true)
   SetLockedOnStateVisuals() {is=title K2Node_FunctionEntry}
   > n5 = SetVisibility(self=Get AnimBoundTrianglesBorder() -> (AnimBoundTrianglesBorder:Border Object Reference), InVisibility=SelfHitTestInvisible)
  ]}
 SetLockingOnStateVisuals {nodes [
   SetLockingOnStateVisuals() {is=title K2Node_FunctionEntry}
   > SetVisibility(self=Get AnimBoundTrianglesBorder(), InVisibility=SelfHitTestInvisible)
   > SetActiveWidget(self=Get Text(), Widget=Get LockingOnText() -> (LockingOnText:Border Object Reference))
   > SetActiveWidget(self=Get CompletedLockonTextSwitcher(), Widget=Get PercentText())
   > SetVisibility(self=Get Percent(), InVisibility=SelfHitTestInvisible)
   > n1 = GetDynamicMaterial(self=Get Progress())
   > SetScalarParameterValue(self=@n1, ParameterName=IsLockingOn, Value=1)
   > n2 = GetDynamicMaterial(self=Get Progress_Container())
   > SetScalarParameterValue(self=@n2, ParameterName=IsLockingOn, Value=1)
   > n3 = GetDynamicMaterial(self=Get Progress())
   > SetScalarParameterValue(self=@n3, ParameterName=IsLockedOn, Value=0)
   > n4 = GetDynamicMaterial(self=Get Progress_Container())
   > SetScalarParameterValue(self=@n4, ParameterName=IsLockedOn, Value=0)
   > SetColorsLockingOn() {is=title K2Node_CallFunction}
   > SetSquareBracketsColors(IsCompleted=0)
   > Set LockingOnTrianglesAnimHasPlayed(LockingOnTrianglesAnimHasPlayed=false)
   > n5 = Branch(Condition=Get LockingOnAnimHasPlayed() -> (LockingOnAnimHasPlayed:bool))
   PlayAnimationForward(execute=@n5.else, InAnimation=Get Anim_LockingOnStarted_Triangles() -> (Anim_LockingOnStarted_Triangles:Widget Animation Object Reference))
   > Set LockingOnAnimHasPlayed(LockingOnAnimHasPlayed=true)
  ]}
 SetOutOfAmmoStateVisuals {nodes [
   SetOutOfAmmoStateVisuals() {is=title K2Node_FunctionEntry}
   > SetVisibility(self=Get AnimBoundTrianglesBorder(), InVisibility=Collapsed)
   > SetActiveWidget(self=Get Text(), Widget=Get NoAmmoText() -> (NoAmmoText:Border Object Reference))
   > SetActiveWidget(self=Get CompletedLockonTextSwitcher(), Widget=Get PercentText())
   > SetVisibility(self=Get Percent(), InVisibility=Collapsed)
   > n1 = GetDynamicMaterial(self=Get Progress())
   > SetScalarParameterValue(self=@n1, ParameterName=Progress, Value=1)
   > n2 = GetDynamicMaterial(self=Get Progress())
   > SetScalarParameterValue(self=@n2, ParameterName=IsLockedOn, Value=1)
   > n3 = GetDynamicMaterial(self=Get Progress_Container())
   > SetScalarParameterValue(self=@n3, ParameterName=IsLockedOn, Value=1)
   > n4 = GetDynamicMaterial(self=Get Progress())
   > SetScalarParameterValue(self=@n4, ParameterName=IsLockingOn, Value=1)
   > n5 = GetDynamicMaterial(self=Get Progress_Container())
   > SetScalarParameterValue(self=@n5, ParameterName=IsLockingOn, Value=1)
   > SetSquareBracketsColors(IsCompleted=1)
   > SetColorsOutOfAmmo() {is=title K2Node_CallFunction}
   > Set LockingOnAnimHasPlayed(LockingOnAnimHasPlayed=false)
  ]}
 SetSearchingStateVisuals {nodes [
   SetSearchingStateVisuals() {is=title K2Node_FunctionEntry}
   > SetVisibility(self=Get AnimBoundTrianglesBorder(), InVisibility=Collapsed)
   > SetActiveWidget(self=Get Text(), Widget=Get SearchingText() -> (SearchingText:Border Object Reference))
   > SetActiveWidget(self=Get CompletedLockonTextSwitcher(), Widget=Get PercentText())
   > SetVisibility(self=Get Percent(), InVisibility=Collapsed)
   > n1 = GetDynamicMaterial(self=Get Progress())
   > SetScalarParameterValue(self=@n1, ParameterName=Progress, Value=0)
   > n2 = GetDynamicMaterial(self=Get Progress())
   > SetScalarParameterValue(self=@n2, ParameterName=IsLockedOn, Value=0)
   > n3 = GetDynamicMaterial(self=Get Progress_Container())
   > SetScalarParameterValue(self=@n3, ParameterName=IsLockedOn, Value=0)
   > n4 = GetDynamicMaterial(self=Get Progress())
   > SetScalarParameterValue(self=@n4, ParameterName=IsLockingOn, Value=0)
   > n5 = GetDynamicMaterial(self=Get Progress_Container())
   > SetScalarParameterValue(self=@n5, ParameterName=IsLockingOn, Value=0)
   > SetSquareBracketsColors(IsCompleted=0)
   > SetColorsSearching() {is=title K2Node_CallFunction}
   > Set LockingOnAnimHasPlayed(LockingOnAnimHasPlayed=false)
  ]}
 SetSquareBracketsColors {nodes [
   n1 = SetSquareBracketsColors!() -> (then, IsCompleted:double=~) {is=title K2Node_FunctionEntry}
   > n2 = For Each Loop(Array=Make Array!(
     "[0]"=Get 1() -> (1:Image Object Reference)
     "[1]"=Get 2() -> (2:Image Object Reference)
     "[2]"=Get 3() -> (3:Image Object Reference)
     "[3]"=Get 4() -> (4:Image Object Reference)
     "[4]":Image Object Reference=Get 5() -> (5:Image Object Reference)
     "[5]":Image Object Reference=Get 6() -> (6:Image Object Reference)
     "[6]":Image Object Reference=Get 7() -> (7:Image Object Reference)
     "[7]":Image Object Reference=Get 8() -> (8:Image Object Reference)
    ) -> (Array))
   > n3 = GetDynamicMaterial(self=@n2)
   > SetScalarParameterValue(self=@n3, ParameterName=CompletedState, Value=@n1)
  ]}
]
variables [
 GlitchFX_ColorA_LockedOnComplete {type=LinearColor, default=<rgba 1 0.075 0.075 1>}
 GlitchFX_ColorA_LockingOn {type=LinearColor, default=<rgba 1 0.6667 0 1>}
 GlitchFX_ColorA_Searching {type=LinearColor, default=<rgba 0 1 0 1>}
 GlitchFX_ColorB_LockedOnComplete {type=LinearColor, default=<rgba 1 0.25 0.2625 1>}
 GlitchFX_ColorB_LockingOn {type=LinearColor, default=<rgba 1 1 0.1 1>}
 GlitchFX_ColorB_Searching {type=LinearColor, default=<rgba 0.5029 1 0.5029 1>}
 GlowLockingColor {type=LinearColor, default=<rgba 1 0.3 0 1>}
 GlowLockingCompleteColor {type=LinearColor, default=<rgba 1 0.025 0.0413 1>}
 GlowSearchingColor {type=LinearColor, default=<rgba 0 1 0 1>}
 HUDElementsToHide {
  type=GameplayTagContainer
  default {gameplayTags [
    {tagName=HUD.TopRight}
    {tagName=HUD.TopLeft.PickupManager}
    {tagName=HUD.TopLeft.PlayerHitPoints}
    {tagName=HUD.TopLeft.PlayerInfo}
    {tagName=HUD.TopLeft.TeamPlayers}
    {tagName=HUD.TopLeft.VoteUpdateNotification}
    {tagName=HUD.Athena.MapPackKeys}
    {tagName=HUD.Athena.GamePhase.MinimapInfo}
   ]}
 }
 IntroAnimHasPlayed {type=boolean}
 LockingOnAnimHasPlayed {type=boolean}
 LockingOnTrianglesAnimHasPlayed {type=boolean}
 New State {type=ENUM_HomingRocket_LockOnStates, default=SearchingForTarget}
 Pip_Color_Glow_LockedOnComplete {type=LinearColor, default=<rgba 1 0.5 0 1>}
 Pip_Color_Glow_LockingOn {type=LinearColor, default=<rgba 1 0.3333 0 1>}
 Pip_Color_Glow_Searching {type=LinearColor, default=<rgba 0 1 0 1>}
 Pip_Color_Pip_LockedOnComplete {type=LinearColor, default=<rgba 1 1 1 1>}
 Pip_Color_Pip_LockingOn {type=LinearColor, default=<rgba 1 0.8375 0.025 1>}
 Pip_Color_Pip_Searching {type=LinearColor, default=<rgba 0.2502 1 0.2502 1>}
]
widgets {
 info {
  inheritedWidgetCount=1
  namedSlotCount=3
  parentClass=&FortHUDElementWidget
  rootWidgetClass=&SizeBox
  widgetCount=105
 }
 widgets [
  All_SizeBox {
   widgetClassPath=&SizeBox
   widget=All_SizeBox
   properties {bOverride_HeightOverride=true, bOverride_WidthOverride=true, heightOverride=520, widthOverride=520}
  }
  All {
   widgetClassPath=&Border
   parent=All_SizeBox
   widget=All
   properties {
    background {
     bIsDynamicallyLoaded=false
     drawAs=Image
     imageType=NoImage
     mirroring=NoMirror
     resourceName=None
     resourceObject=/ResolveGameplay/Items/Guns/HomingRocket/Widgets/UI_Materials/M_UI_HomingRocket_SquareMask
     tiling=NoTile
     imageSize=<32 32>
     margin=<ltrb 0.5 0.5 0.5 0.5>
     outlineSettings {
      bUseBrushTransparency=false
      roundingType=HalfHeightRadius
      width=0
      color {colorUseRule=UseColor_Specified, specifiedColor=<rgba 0 0 0 0>}
      cornerRadii=<0 0 0 0>
     }
     tintColor {colorUseRule=UseColor_Specified, specifiedColor=<rgba 1 1 1 1>}
     uVRegion {bIsValid=false, max=<0 0>, min=<0 0>}
    }
    padding=<ltrb 8 8 8 8>
   }
   slot {class=SizeBoxSlot}
  }
  Overlay_100 {widgetClassPath=&Overlay, parent=All, widget=Overlay_100, slot {class=BorderSlot, properties {padding=<ltrb 8 8 8 8>}}}
  LeftSideMarkers {
   widgetClassPath=&VerticalBox
   parent=Overlay_100
   widget=LeftSideMarkers
   properties {renderTransform {angle=0, scale=<1 1>, shear=<0 0>, translation=<-384 0>}}
   slot {class=OverlaySlot}
  }
  DashesTopLeft {
   widgetClassPath=&Overlay
   parent=LeftSideMarkers
   widget=DashesTopLeft
   properties {renderTransform {angle=0, scale=<1 1>, shear=<0 0>, translation=<0 -16>}}
   slot {class=VerticalBoxSlot, properties {horizontalAlignment=HAlign_Left, verticalAlignment=VAlign_Top}}
  }
  SmallDashTopLeft {
   widgetClassPath=&Image
   parent=DashesTopLeft
   widget=SmallDashTopLeft
   properties {
    brush {
     bIsDynamicallyLoaded=false
     drawAs=Image
     imageType=NoImage
     mirroring=NoMirror
     resourceName=None
     resourceObject=/ResolveGameplay/Items/Guns/HomingRocket/Widgets/UI_Materials/M_UI_Homing_Dashed_Lines_Small
     tiling=NoTile
     imageSize=<24 32>
     margin=<ltrb 0 0 0 0>
     outlineSettings {
      bUseBrushTransparency=false
      roundingType=HalfHeightRadius
      width=0
      color {colorUseRule=UseColor_Specified, specifiedColor=<rgba 0 0 0 0>}
      cornerRadii=<0 0 0 0>
     }
     tintColor {colorUseRule=UseColor_Specified, specifiedColor=<rgba 1 1 1 1>}
     uVRegion {bIsValid=false, max=<0 0>, min=<0 0>}
    }
    renderTransform {angle=0, scale=<1 1>, shear=<0 0>, translation=<0 32>}
   }
   slot {class=OverlaySlot, properties {verticalAlignment=VAlign_Fill}}
  }
  LargeDashTopLeft {
   widgetClassPath=&Image
   parent=DashesTopLeft
   widget=LargeDashTopLeft
   properties {brush {
     bIsDynamicallyLoaded=false
     drawAs=Image
     imageType=NoImage
     mirroring=NoMirror
     resourceName=None
     resourceObject=&M_UI_Homing_Dashed_Lines_Small
     tiling=NoTile
     imageSize=<40 256>
     margin=<ltrb 0 0 0 0>
     outlineSettings {
      bUseBrushTransparency=false
      roundingType=HalfHeightRadius
      width=0
      color {colorUseRule=UseColor_Specified, specifiedColor=<rgba 0 0 0 0>}
      cornerRadii=<0 0 0 0>
     }
     tintColor {colorUseRule=UseColor_Specified, specifiedColor=<rgba 1 1 1 1>}
     uVRegion {bIsValid=false, max=<0 0>, min=<0 0>}
    }}
   slot {class=OverlaySlot}
  }
  CenterLineLeft {
   widgetClassPath=&SizeBox
   parent=LeftSideMarkers
   widget=CenterLineLeft
   properties {
    bOverride_HeightOverride=true
    bOverride_WidthOverride=true
    heightOverride=4
    renderOpacity=0.75
    widthOverride=320
    renderTransformPivot=<0 0.5>
   }
   slot {class=VerticalBoxSlot, properties {horizontalAlignment=HAlign_Left, verticalAlignment=VAlign_Top}}
  }
  Image_14 {
   widgetClassPath=&Image
   parent=CenterLineLeft
   widget=Image_14
   properties {brush {
     bIsDynamicallyLoaded=false
     drawAs=Box
     imageType=NoImage
     mirroring=NoMirror
     resourceName=None
     resourceObject=/ResolveGameplay/Items/Guns/HomingRocket/Widgets/UI_Materials/M_UI_HomingRocket_Crosshair_Centerline
     tiling=NoTile
     imageSize=<8 8>
     margin=<ltrb 0.5 0.5 0.5 0.5>
     outlineSettings {
      bUseBrushTransparency=false
      roundingType=HalfHeightRadius
      width=0
      color {colorUseRule=UseColor_Specified, specifiedColor=<rgba 0 0 0 0>}
      cornerRadii=<0 0 0 0>
     }
     tintColor {colorUseRule=UseColor_Specified, specifiedColor=<rgba 1 1 1 1>}
     uVRegion {bIsValid=false, max=<0 0>, min=<0 0>}
    }}
   slot {class=SizeBoxSlot}
  }
  DashesBottomLeft {
   widgetClassPath=&Overlay
   parent=LeftSideMarkers
   widget=DashesBottomLeft
   properties {renderTransform {angle=0, scale=<1 1>, shear=<0 0>, translation=<0 16>}}
   slot {class=VerticalBoxSlot, properties {horizontalAlignment=HAlign_Left, verticalAlignment=VAlign_Top}}
  }
  SmallDashBottomLeft {
   widgetClassPath=&Image
   parent=DashesBottomLeft
   widget=SmallDashBottomLeft
   properties {brush {
     bIsDynamicallyLoaded=false
     drawAs=Image
     imageType=NoImage
     mirroring=NoMirror
     resourceName=None
     resourceObject=&M_UI_Homing_Dashed_Lines_Small
     tiling=NoTile
     imageSize=<24 32>
     margin=<ltrb 0 0 0 0>
     outlineSettings {
      bUseBrushTransparency=false
      roundingType=HalfHeightRadius
      width=0
      color {colorUseRule=UseColor_Specified, specifiedColor=<rgba 0 0 0 0>}
      cornerRadii=<0 0 0 0>
     }
     tintColor {colorUseRule=UseColor_Specified, specifiedColor=<rgba 1 1 1 1>}
     uVRegion {bIsValid=false, max=<0 0>, min=<0 0>}
    }}
   slot {class=OverlaySlot, properties {verticalAlignment=VAlign_Fill}}
  }
  LargeDashBottomLeft {
   widgetClassPath=&Image
   parent=DashesBottomLeft
   widget=LargeDashBottomLeft
   properties {
    brush {
     bIsDynamicallyLoaded=false
     drawAs=Image
     imageType=NoImage
     mirroring=NoMirror
     resourceName=None
     resourceObject=&M_UI_Homing_Dashed_Lines_Small
     tiling=NoTile
     imageSize=<40 256>
     margin=<ltrb 0 0 0 0>
     outlineSettings {
      bUseBrushTransparency=false
      roundingType=HalfHeightRadius
      width=0
      color {colorUseRule=UseColor_Specified, specifiedColor=<rgba 0 0 0 0>}
      cornerRadii=<0 0 0 0>
     }
     tintColor {colorUseRule=UseColor_Specified, specifiedColor=<rgba 1 1 1 1>}
     uVRegion {bIsValid=false, max=<0 0>, min=<0 0>}
    }
    renderTransform {angle=0, scale=<1 1>, shear=<0 0>, translation=<0 32>}
   }
   slot {class=OverlaySlot}
  }
  RightSideMarkers {
   widgetClassPath=&VerticalBox
   parent=Overlay_100
   widget=RightSideMarkers
   properties {renderTransform {angle=0, scale=<1 1>, shear=<0 0>, translation=<384 0>}}
   slot {class=OverlaySlot, properties {horizontalAlignment=HAlign_Right}}
  }
  DashesTopRight {
   widgetClassPath=&Overlay
   parent=RightSideMarkers
   widget=DashesTopRight
   properties {renderTransform {angle=0, scale=<1 1>, shear=<0 0>, translation=<0 -16>}}
   slot {class=VerticalBoxSlot, properties {horizontalAlignment=HAlign_Right, verticalAlignment=VAlign_Top}}
  }
  SmallDashTopRight {
   widgetClassPath=&Image
   parent=DashesTopRight
   widget=SmallDashTopRight
   properties {
    brush {
     bIsDynamicallyLoaded=false
     drawAs=Image
     imageType=NoImage
     mirroring=NoMirror
     resourceName=None
     resourceObject=&M_UI_Homing_Dashed_Lines_Small
     tiling=NoTile
     imageSize=<24 32>
     margin=<ltrb 0 0 0 0>
     outlineSettings {
      bUseBrushTransparency=false
      roundingType=HalfHeightRadius
      width=0
      color {colorUseRule=UseColor_Specified, specifiedColor=<rgba 0 0 0 0>}
      cornerRadii=<0 0 0 0>
     }
     tintColor {colorUseRule=UseColor_Specified, specifiedColor=<rgba 1 1 1 1>}
     uVRegion {bIsValid=false, max=<0 0>, min=<0 0>}
    }
    renderTransform {angle=0, scale=<1 1>, shear=<0 0>, translation=<0 32>}
   }
   slot {class=OverlaySlot, properties {horizontalAlignment=HAlign_Right, verticalAlignment=VAlign_Fill}}
  }
  LargeDashTopRight {
   widgetClassPath=&Image
   parent=DashesTopRight
   widget=LargeDashTopRight
   properties {brush {
     bIsDynamicallyLoaded=false
     drawAs=Image
     imageType=NoImage
     mirroring=NoMirror
     resourceName=None
     resourceObject=&M_UI_Homing_Dashed_Lines_Small
     tiling=NoTile
     imageSize=<40 256>
     margin=<ltrb 0 0 0 0>
     outlineSettings {
      bUseBrushTransparency=false
      roundingType=HalfHeightRadius
      width=0
      color {colorUseRule=UseColor_Specified, specifiedColor=<rgba 0 0 0 0>}
      cornerRadii=<0 0 0 0>
     }
     tintColor {colorUseRule=UseColor_Specified, specifiedColor=<rgba 1 1 1 1>}
     uVRegion {bIsValid=false, max=<0 0>, min=<0 0>}
    }}
   slot {class=OverlaySlot, properties {horizontalAlignment=HAlign_Right}}
  }
  CenterLineRight {
   widgetClassPath=&SizeBox
   parent=RightSideMarkers
   widget=CenterLineRight
   properties {
    bOverride_HeightOverride=true
    bOverride_WidthOverride=true
    heightOverride=4
    renderOpacity=0.75
    widthOverride=320
    renderTransformPivot=<1 0.5>
   }
   slot {class=VerticalBoxSlot, properties {horizontalAlignment=HAlign_Right, verticalAlignment=VAlign_Center}}
  }
  Image_29 {
   widgetClassPath=&Image
   parent=CenterLineRight
   widget=Image_29
   properties {brush {
     bIsDynamicallyLoaded=false
     drawAs=Box
     imageType=NoImage
     mirroring=NoMirror
     resourceName=None
     resourceObject=&M_UI_HomingRocket_Crosshair_Centerline
     tiling=NoTile
     imageSize=<8 8>
     margin=<ltrb 0.5 0.5 0.5 0.5>
     outlineSettings {
      bUseBrushTransparency=false
      roundingType=HalfHeightRadius
      width=0
      color {colorUseRule=UseColor_Specified, specifiedColor=<rgba 0 0 0 0>}
      cornerRadii=<0 0 0 0>
     }
     tintColor {colorUseRule=UseColor_Specified, specifiedColor=<rgba 1 1 1 1>}
     uVRegion {bIsValid=false, max=<0 0>, min=<0 0>}
    }}
   slot {class=SizeBoxSlot}
  }
  DashesBottomRight {
   widgetClassPath=&Overlay
   parent=RightSideMarkers
   widget=DashesBottomRight
   properties {renderTransform {angle=0, scale=<1 1>, shear=<0 0>, translation=<0 16>}}
   slot {class=VerticalBoxSlot, properties {horizontalAlignment=HAlign_Right, verticalAlignment=VAlign_Top}}
  }
  SmallDashBottomRight {
   widgetClassPath=&Image
   parent=DashesBottomRight
   widget=SmallDashBottomRight
   properties {brush {
     bIsDynamicallyLoaded=false
     drawAs=Image
     imageType=NoImage
     mirroring=NoMirror
     resourceName=None
     resourceObject=&M_UI_Homing_Dashed_Lines_Small
     tiling=NoTile
     imageSize=<24 32>
     margin=<ltrb 0 0 0 0>
     outlineSettings {
      bUseBrushTransparency=false
      roundingType=HalfHeightRadius
      width=0
      color {colorUseRule=UseColor_Specified, specifiedColor=<rgba 0 0 0 0>}
      cornerRadii=<0 0 0 0>
     }
     tintColor {colorUseRule=UseColor_Specified, specifiedColor=<rgba 1 1 1 1>}
     uVRegion {bIsValid=false, max=<0 0>, min=<0 0>}
    }}
   slot {class=OverlaySlot, properties {horizontalAlignment=HAlign_Right, verticalAlignment=VAlign_Fill}}
  }
  LargeDashBottomRight {
   widgetClassPath=&Image
   parent=DashesBottomRight
   widget=LargeDashBottomRight
   properties {
    brush {
     bIsDynamicallyLoaded=false
     drawAs=Image
     imageType=NoImage
     mirroring=NoMirror
     resourceName=None
     resourceObject=&M_UI_Homing_Dashed_Lines_Small
     tiling=NoTile
     imageSize=<40 256>
     margin=<ltrb 0 0 0 0>
     outlineSettings {
      bUseBrushTransparency=false
      roundingType=HalfHeightRadius
      width=0
      color {colorUseRule=UseColor_Specified, specifiedColor=<rgba 0 0 0 0>}
      cornerRadii=<0 0 0 0>
     }
     tintColor {colorUseRule=UseColor_Specified, specifiedColor=<rgba 1 1 1 1>}
     uVRegion {bIsValid=false, max=<0 0>, min=<0 0>}
    }
    renderTransform {angle=0, scale=<1 1>, shear=<0 0>, translation=<0 32>}
   }
   slot {class=OverlaySlot, properties {horizontalAlignment=HAlign_Right}}
  }
  FlavorGraphics {
   widgetClassPath=&HorizontalBox
   parent=Overlay_100
   widget=FlavorGraphics
   properties {visibility=Collapsed, renderTransform {angle=0, scale=<1 1>, shear=<0 0>, translation=<0 -192>}}
   slot {class=OverlaySlot, properties {horizontalAlignment=HAlign_Fill}}
  }
  SeekGraphic {
   widgetClassPath=&Overlay
   parent=FlavorGraphics
   widget=SeekGraphic
   properties {renderTransform {angle=0, scale=<1 1>, shear=<0 0>, translation=<-256 0>}}
   slot {class=HorizontalBoxSlot}
  }
  LeftGlow {
   widgetClassPath=&Image
   parent=SeekGraphic
   widget=LeftGlow
   properties {brush {
     bIsDynamicallyLoaded=false
     drawAs=Image
     imageType=NoImage
     mirroring=NoMirror
     resourceName=None
     resourceObject=/ResolveGameplay/Items/Guns/HomingRocket/Widgets/UI_Materials/M_UI_Homing_Ring_Glow
     tiling=NoTile
     imageSize=<168 168>
     margin=<ltrb 0 0 0 0>
     outlineSettings {
      bUseBrushTransparency=false
      roundingType=HalfHeightRadius
      width=0
      color {colorUseRule=UseColor_Specified, specifiedColor=<rgba 0 0 0 0>}
      cornerRadii=<0 0 0 0>
     }
     tintColor {colorUseRule=UseColor_Specified, specifiedColor=<rgba 1 1 1 1>}
     uVRegion {bIsValid=false, max=<0 0>, min=<0 0>}
    }}
   slot {class=OverlaySlot, properties {horizontalAlignment=HAlign_Fill, verticalAlignment=VAlign_Fill}}
  }
  LeftRing {
   widgetClassPath=&Image
   parent=SeekGraphic
   widget=LeftRing
   properties {brush {
     bIsDynamicallyLoaded=false
     drawAs=Image
     imageType=NoImage
     mirroring=NoMirror
     resourceName=None
     resourceObject=/ResolveGameplay/Items/Guns/HomingRocket/Widgets/UI_Materials/M_UI_Homing_Ring
     tiling=NoTile
     imageSize=<168 168>
     margin=<ltrb 0 0 0 0>
     outlineSettings {
      bUseBrushTransparency=false
      roundingType=HalfHeightRadius
      width=0
      color {colorUseRule=UseColor_Specified, specifiedColor=<rgba 0 0 0 0>}
      cornerRadii=<0 0 0 0>
     }
     tintColor {colorUseRule=UseColor_Specified, specifiedColor=<rgba 1 1 1 1>}
     uVRegion {bIsValid=false, max=<0 0>, min=<0 0>}
    }}
   slot {class=OverlaySlot, properties {horizontalAlignment=HAlign_Fill, verticalAlignment=VAlign_Fill}}
  }
  FortHUDElementWrapper_TopText {
   widgetClassPath=/Game/Athena/HUD/FortHUDElementWrapper
   bIsVariable=true
   parent=SeekGraphic
   widget=FortHUDElementWrapper_TopText
   properties {elementTag {tagName=HUD.Resolve.HomingRocket.LockOnLeftText}}
   slot {class=OverlaySlot, properties {horizontalAlignment=HAlign_Center, verticalAlignment=VAlign_Center}}
  }
  CommonTextBlock_62 {
   widgetClassPath=&CommonTextBlock
   namedSlotHost=FortHUDElementWrapper_TopText
   widget=CommonTextBlock_62
   properties {
    text=Top
    textTransformPolicy=ToUpper
    font {
     bForceMonospaced=false
     bMaterialIsStencil=false
     fontMaterial=/ResolveGameplay/Items/Guns/HomingRocket/Widgets/UI_Materials/M_UI_Homing_Font_Flavor
     fontObject=&Burbank
     letterSpacing=384
     monospacedWidth=1
     size=13
     skewAmount=0
     typefaceFontName=BurbankRegular-Black
     outlineSettings {
      bApplyOutlineToDropShadows=false
      bMiteredCorners=false
      bSeparateFillAlpha=false
      outlineMaterial=null
      outlineSize=0
      outlineColor=<rgba 0 0 0 1>
     }
    }
    shadowOffset=<0 0>
   }
  }
  TopGraphic {
   widgetClassPath=&Overlay
   parent=FlavorGraphics
   widget=TopGraphic
   properties {renderTransform {angle=0, scale=<1 1>, shear=<0 0>, translation=<256 0>}}
   slot {class=HorizontalBoxSlot, properties {horizontalAlignment=HAlign_Right, verticalAlignment=VAlign_Top, size {sizeRule=Fill, value=1}}}
  }
  RightRingGlow {
   widgetClassPath=&Image
   parent=TopGraphic
   widget=RightRingGlow
   properties {brush {
     bIsDynamicallyLoaded=false
     drawAs=Image
     imageType=NoImage
     mirroring=NoMirror
     resourceName=None
     resourceObject=&M_UI_Homing_Ring_Glow
     tiling=NoTile
     imageSize=<168 168>
     margin=<ltrb 0 0 0 0>
     outlineSettings {
      bUseBrushTransparency=false
      roundingType=HalfHeightRadius
      width=0
      color {colorUseRule=UseColor_Specified, specifiedColor=<rgba 0 0 0 0>}
      cornerRadii=<0 0 0 0>
     }
     tintColor {colorUseRule=UseColor_Specified, specifiedColor=<rgba 1 1 1 1>}
     uVRegion {bIsValid=false, max=<0 0>, min=<0 0>}
    }}
   slot {class=OverlaySlot, properties {horizontalAlignment=HAlign_Fill, verticalAlignment=VAlign_Fill}}
  }
  RightRing {
   widgetClassPath=&Image
   parent=TopGraphic
   widget=RightRing
   properties {brush {
     bIsDynamicallyLoaded=false
     drawAs=Image
     imageType=NoImage
     mirroring=NoMirror
     resourceName=None
     resourceObject=&M_UI_Homing_Ring
     tiling=NoTile
     imageSize=<168 168>
     margin=<ltrb 0 0 0 0>
     outlineSettings {
      bUseBrushTransparency=false
      roundingType=HalfHeightRadius
      width=0
      color {colorUseRule=UseColor_Specified, specifiedColor=<rgba 0 0 0 0>}
      cornerRadii=<0 0 0 0>
     }
     tintColor {colorUseRule=UseColor_Specified, specifiedColor=<rgba 1 1 1 1>}
     uVRegion {bIsValid=false, max=<0 0>, min=<0 0>}
    }}
   slot {class=OverlaySlot, properties {horizontalAlignment=HAlign_Fill, verticalAlignment=VAlign_Fill}}
  }
  FortHUDElementWrapper_SeekText {
   widgetClassPath=&FortHUDElementWrapper
   bIsVariable=true
   parent=TopGraphic
   widget=FortHUDElementWrapper_SeekText
   properties {elementTag {tagName=HUD.Resolve.HomingRocket.LockOnRightText}}
   slot {class=OverlaySlot, properties {horizontalAlignment=HAlign_Center, verticalAlignment=VAlign_Center}}
  }
  CommonTextBlock_1 {
   widgetClassPath=&CommonTextBlock
   namedSlotHost=FortHUDElementWrapper_SeekText
   widget=CommonTextBlock_1
   properties {
    text=Seek
    textTransformPolicy=ToUpper
    font {
     bForceMonospaced=false
     bMaterialIsStencil=false
     fontMaterial=&M_UI_Homing_Font_Flavor
     fontObject=&Burbank
     letterSpacing=384
     monospacedWidth=1
     size=13
     skewAmount=0
     typefaceFontName=BurbankRegular-Black
     outlineSettings {
      bApplyOutlineToDropShadows=false
      bMiteredCorners=false
      bSeparateFillAlpha=false
      outlineMaterial=null
      outlineSize=0
      outlineColor=<rgba 0 0 0 1>
     }
    }
    shadowOffset=<0 0>
   }
  }
  FlavorText {
   widgetClassPath=&HorizontalBox
   parent=Overlay_100
   widget=FlavorText
   properties {renderTransform {angle=0, scale=<1 1>, shear=<0 0>, translation=<0 -24>}}
   slot {class=OverlaySlot, properties {horizontalAlignment=HAlign_Fill}}
  }
  CommonTextBlock_65 {
   widgetClassPath=&CommonTextBlock
   parent=FlavorText
   widget=CommonTextBlock_65
   properties {
    text=6G-3Z
    textTransformPolicy=ToUpper
    font {
     bForceMonospaced=false
     bMaterialIsStencil=false
     fontMaterial=&M_UI_Homing_Font_Flavor
     fontObject=&BurbankSmall
     letterSpacing=256
     monospacedWidth=1
     size=11
     skewAmount=0
     typefaceFontName=Bold
     outlineSettings {
      bApplyOutlineToDropShadows=false
      bMiteredCorners=false
      bSeparateFillAlpha=false
      outlineMaterial=null
      outlineSize=0
      outlineColor=<rgba 0 0 0 1>
     }
    }
    shadowOffset=<0 0>
   }
   slot {class=HorizontalBoxSlot}
  }
  CommonTextBlock {
   widgetClassPath=&CommonTextBlock
   parent=FlavorText
   widget=CommonTextBlock
   properties {
    justification=Right
    text="[ 2 x ]"
    font {
     bForceMonospaced=false
     bMaterialIsStencil=false
     fontMaterial=&M_UI_Homing_Font_Flavor
     fontObject=&BurbankSmall
     letterSpacing=64
     monospacedWidth=1
     size=11
     skewAmount=0
     typefaceFontName=Bold
     outlineSettings {
      bApplyOutlineToDropShadows=false
      bMiteredCorners=false
      bSeparateFillAlpha=false
      outlineMaterial=null
      outlineSize=0
      outlineColor=<rgba 0 0 0 1>
     }
    }
    shadowOffset=<0 0>
   }
   slot {class=HorizontalBoxSlot, properties {horizontalAlignment=HAlign_Right, size {sizeRule=Fill, value=1}}}
  }
  BottomDashedLines {
   widgetClassPath=&HorizontalBox
   parent=Overlay_100
   widget=BottomDashedLines
   properties {renderOpacity=0.325, renderTransform {angle=0, scale=<1 1>, shear=<0 0>, translation=<0 192>}}
   slot {class=OverlaySlot, properties {horizontalAlignment=HAlign_Center, verticalAlignment=VAlign_Bottom}}
  }
  Image_39 {
   widgetClassPath=&Image
   parent=BottomDashedLines
   widget=Image_39
   properties {
    renderOpacity=0.5
    brush {
     bIsDynamicallyLoaded=false
     drawAs=Image
     imageType=NoImage
     mirroring=NoMirror
     resourceName=None
     resourceObject=/Game/UI/Foundation/Textures/Rule/Rule-Dotted-HUD
     tiling=Vertical
     imageSize=<8 48>
     margin=<ltrb 0 0 0 0>
     outlineSettings {
      bUseBrushTransparency=false
      roundingType=HalfHeightRadius
      width=0
      color {colorUseRule=UseColor_Specified, specifiedColor=<rgba 0 0 0 0>}
      cornerRadii=<0 0 0 0>
     }
     tintColor {colorUseRule=UseColor_Specified, specifiedColor=<rgba 0.25 1 0.25 1>}
     uVRegion {bIsValid=false, max=<0 0>, min=<0 0>}
    }
   }
   slot {class=HorizontalBoxSlot, properties {verticalAlignment=VAlign_Bottom}}
  }
  Image_40 {
   widgetClassPath=&Image
   parent=BottomDashedLines
   widget=Image_40
   properties {brush {
     bIsDynamicallyLoaded=false
     drawAs=Image
     imageType=NoImage
     mirroring=NoMirror
     resourceName=None
     resourceObject=&Rule-Dotted-HUD
     tiling=Vertical
     imageSize=<8 88>
     margin=<ltrb 0 0 0 0>
     outlineSettings {
      bUseBrushTransparency=false
      roundingType=HalfHeightRadius
      width=0
      color {colorUseRule=UseColor_Specified, specifiedColor=<rgba 0 0 0 0>}
      cornerRadii=<0 0 0 0>
     }
     tintColor {colorUseRule=UseColor_Specified, specifiedColor=<rgba 0.334 1 0.334 1>}
     uVRegion {bIsValid=false, max=<0 0>, min=<0 0>}
    }}
   slot {class=HorizontalBoxSlot, properties {verticalAlignment=VAlign_Bottom, padding=<ltrb 16 0 16 0>}}
  }
  Image_41 {
   widgetClassPath=&Image
   parent=BottomDashedLines
   widget=Image_41
   properties {
    renderOpacity=0.5
    brush {
     bIsDynamicallyLoaded=false
     drawAs=Image
     imageType=NoImage
     mirroring=NoMirror
     resourceName=None
     resourceObject=&Rule-Dotted-HUD
     tiling=Vertical
     imageSize=<8 48>
     margin=<ltrb 0 0 0 0>
     outlineSettings {
      bUseBrushTransparency=false
      roundingType=HalfHeightRadius
      width=0
      color {colorUseRule=UseColor_Specified, specifiedColor=<rgba 0 0 0 0>}
      cornerRadii=<0 0 0 0>
     }
     tintColor {colorUseRule=UseColor_Specified, specifiedColor=<rgba 0.25 1 0.25 1>}
     uVRegion {bIsValid=false, max=<0 0>, min=<0 0>}
    }
   }
   slot {class=HorizontalBoxSlot, properties {verticalAlignment=VAlign_Bottom}}
  }
  TopDashedLines {
   widgetClassPath=&HorizontalBox
   parent=Overlay_100
   widget=TopDashedLines
   properties {renderOpacity=0.325, renderTransform {angle=0, scale=<1 1>, shear=<0 0>, translation=<0 -192>}}
   slot {class=OverlaySlot, properties {horizontalAlignment=HAlign_Center}}
  }
  Image_489 {
   widgetClassPath=&Image
   parent=TopDashedLines
   widget=Image_489
   properties {
    renderOpacity=0.5
    brush {
     bIsDynamicallyLoaded=false
     drawAs=Image
     imageType=NoImage
     mirroring=NoMirror
     resourceName=None
     resourceObject=&Rule-Dotted-HUD
     tiling=Vertical
     imageSize=<8 48>
     margin=<ltrb 0 0 0 0>
     outlineSettings {
      bUseBrushTransparency=false
      roundingType=HalfHeightRadius
      width=0
      color {colorUseRule=UseColor_Specified, specifiedColor=<rgba 0 0 0 0>}
      cornerRadii=<0 0 0 0>
     }
     tintColor {colorUseRule=UseColor_Specified, specifiedColor=<rgba 0.25 1 0.25 1>}
     uVRegion {bIsValid=false, max=<0 0>, min=<0 0>}
    }
   }
   slot {class=HorizontalBoxSlot, properties {verticalAlignment=VAlign_Top}}
  }
  Image_37 {
   widgetClassPath=&Image
   parent=TopDashedLines
   widget=Image_37
   properties {brush {
     bIsDynamicallyLoaded=false
     drawAs=Image
     imageType=NoImage
     mirroring=NoMirror
     resourceName=None
     resourceObject=&Rule-Dotted-HUD
     tiling=Vertical
     imageSize=<8 88>
     margin=<ltrb 0 0 0 0>
     outlineSettings {
      bUseBrushTransparency=false
      roundingType=HalfHeightRadius
      width=0
      color {colorUseRule=UseColor_Specified, specifiedColor=<rgba 0 0 0 0>}
      cornerRadii=<0 0 0 0>
     }
     tintColor {colorUseRule=UseColor_Specified, specifiedColor=<rgba 0.334 1 0.334 1>}
     uVRegion {bIsValid=false, max=<0 0>, min=<0 0>}
    }}
   slot {class=HorizontalBoxSlot, properties {verticalAlignment=VAlign_Top, padding=<ltrb 16 0 16 0>}}
  }
  Image_38 {
   widgetClassPath=&Image
   parent=TopDashedLines
   widget=Image_38
   properties {
    renderOpacity=0.5
    brush {
     bIsDynamicallyLoaded=false
     drawAs=Image
     imageType=NoImage
     mirroring=NoMirror
     resourceName=None
     resourceObject=&Rule-Dotted-HUD
     tiling=Vertical
     imageSize=<8 48>
     margin=<ltrb 0 0 0 0>
     outlineSettings {
      bUseBrushTransparency=false
      roundingType=HalfHeightRadius
      width=0
      color {colorUseRule=UseColor_Specified, specifiedColor=<rgba 0 0 0 0>}
      cornerRadii=<0 0 0 0>
     }
     tintColor {colorUseRule=UseColor_Specified, specifiedColor=<rgba 0.25 1 0.25 1>}
     uVRegion {bIsValid=false, max=<0 0>, min=<0 0>}
    }
   }
   slot {class=HorizontalBoxSlot, properties {verticalAlignment=VAlign_Top}}
  }
  SquareBrackets {
   widgetClassPath=&Overlay
   parent=Overlay_100
   widget=SquareBrackets
   slot {class=OverlaySlot, properties {horizontalAlignment=HAlign_Fill, verticalAlignment=VAlign_Fill}}
  }
  TopLeft {
   widgetClassPath=&Overlay
   parent=SquareBrackets
   widget=TopLeft
   properties {renderTransform {angle=0, scale=<1 1>, shear=<0 0>, translation=<-2 -2>}}
   slot {class=OverlaySlot}
  }
  SizeBox_1 {
   widgetClassPath=&SizeBox
   parent=TopLeft
   widget=SizeBox_1
   properties {bOverride_HeightOverride=true, bOverride_WidthOverride=true, heightOverride=80, widthOverride=8}
   slot {class=OverlaySlot}
  }
  1 {
   widgetClassPath=&Image
   bIsVariable=true
   parent=SizeBox_1
   widget=1
   properties {brush {
     bIsDynamicallyLoaded=false
     drawAs=Box
     imageType=NoImage
     mirroring=NoMirror
     resourceName=None
     resourceObject=/ResolveGameplay/Items/Guns/HomingRocket/Widgets/UI_Materials/M_UI_HomingRocket_Crosshair_Border
     tiling=NoTile
     imageSize=<8 8>
     margin=<ltrb 0.5 0.5 0.5 0.5>
     outlineSettings {
      bUseBrushTransparency=false
      roundingType=HalfHeightRadius
      width=0
      color {colorUseRule=UseColor_Specified, specifiedColor=<rgba 0 0 0 0>}
      cornerRadii=<0 0 0 0>
     }
     tintColor {colorUseRule=UseColor_Specified, specifiedColor=<rgba 1 1 1 1>}
     uVRegion {bIsValid=false, max=<0 0>, min=<0 0>}
    }}
   slot {class=SizeBoxSlot}
  }
  SizeBox {
   widgetClassPath=&SizeBox
   parent=TopLeft
   widget=SizeBox
   properties {bOverride_HeightOverride=true, bOverride_WidthOverride=true, heightOverride=8, widthOverride=80}
   slot {class=OverlaySlot}
  }
  2 {
   widgetClassPath=&Image
   bIsVariable=true
   parent=SizeBox
   widget=2
   properties {brush {
     bIsDynamicallyLoaded=false
     drawAs=Box
     imageType=NoImage
     mirroring=NoMirror
     resourceName=None
     resourceObject=&M_UI_HomingRocket_Crosshair_Border
     tiling=NoTile
     imageSize=<8 8>
     margin=<ltrb 0.5 0.5 0.5 0.5>
     outlineSettings {
      bUseBrushTransparency=false
      roundingType=HalfHeightRadius
      width=0
      color {colorUseRule=UseColor_Specified, specifiedColor=<rgba 0 0 0 0>}
      cornerRadii=<0 0 0 0>
     }
     tintColor {colorUseRule=UseColor_Specified, specifiedColor=<rgba 1 1 1 1>}
     uVRegion {bIsValid=false, max=<0 0>, min=<0 0>}
    }}
   slot {class=SizeBoxSlot}
  }
  BottomLeft {
   widgetClassPath=&Overlay
   parent=SquareBrackets
   widget=BottomLeft
   properties {renderTransform {angle=-90, scale=<1 1>, shear=<0 0>, translation=<-2 2>}}
   slot {class=OverlaySlot, properties {verticalAlignment=VAlign_Bottom}}
  }
  SizeBox_4 {
   widgetClassPath=&SizeBox
   parent=BottomLeft
   widget=SizeBox_4
   properties {bOverride_HeightOverride=true, bOverride_WidthOverride=true, heightOverride=80, widthOverride=8}
   slot {class=OverlaySlot}
  }
  3 {
   widgetClassPath=&Image
   bIsVariable=true
   parent=SizeBox_4
   widget=3
   properties {brush {
     bIsDynamicallyLoaded=false
     drawAs=Box
     imageType=NoImage
     mirroring=NoMirror
     resourceName=None
     resourceObject=&M_UI_HomingRocket_Crosshair_Border
     tiling=NoTile
     imageSize=<8 8>
     margin=<ltrb 0.5 0.5 0.5 0.5>
     outlineSettings {
      bUseBrushTransparency=false
      roundingType=HalfHeightRadius
      width=0
      color {colorUseRule=UseColor_Specified, specifiedColor=<rgba 0 0 0 0>}
      cornerRadii=<0 0 0 0>
     }
     tintColor {colorUseRule=UseColor_Specified, specifiedColor=<rgba 1 1 1 1>}
     uVRegion {bIsValid=false, max=<0 0>, min=<0 0>}
    }}
   slot {class=SizeBoxSlot}
  }
  SizeBox_5 {
   widgetClassPath=&SizeBox
   parent=BottomLeft
   widget=SizeBox_5
   properties {bOverride_HeightOverride=true, bOverride_WidthOverride=true, heightOverride=8, widthOverride=80}
   slot {class=OverlaySlot}
  }
  4 {
   widgetClassPath=&Image
   bIsVariable=true
   parent=SizeBox_5
   widget=4
   properties {brush {
     bIsDynamicallyLoaded=false
     drawAs=Box
     imageType=NoImage
     mirroring=NoMirror
     resourceName=None
     resourceObject=&M_UI_HomingRocket_Crosshair_Border
     tiling=NoTile
     imageSize=<8 8>
     margin=<ltrb 0.5 0.5 0.5 0.5>
     outlineSettings {
      bUseBrushTransparency=false
      roundingType=HalfHeightRadius
      width=0
      color {colorUseRule=UseColor_Specified, specifiedColor=<rgba 0 0 0 0>}
      cornerRadii=<0 0 0 0>
     }
     tintColor {colorUseRule=UseColor_Specified, specifiedColor=<rgba 1 1 1 1>}
     uVRegion {bIsValid=false, max=<0 0>, min=<0 0>}
    }}
   slot {class=SizeBoxSlot}
  }
  TopRight {
   widgetClassPath=&Overlay
   parent=SquareBrackets
   widget=TopRight
   properties {renderTransform {angle=90, scale=<1 1>, shear=<0 0>, translation=<2 -2>}}
   slot {class=OverlaySlot, properties {horizontalAlignment=HAlign_Right}}
  }
  SizeBox_2 {
   widgetClassPath=&SizeBox
   parent=TopRight
   widget=SizeBox_2
   properties {bOverride_HeightOverride=true, bOverride_WidthOverride=true, heightOverride=80, widthOverride=8}
   slot {class=OverlaySlot}
  }
  5 {
   widgetClassPath=&Image
   bIsVariable=true
   parent=SizeBox_2
   widget=5
   properties {brush {
     bIsDynamicallyLoaded=false
     drawAs=Box
     imageType=NoImage
     mirroring=NoMirror
     resourceName=None
     resourceObject=&M_UI_HomingRocket_Crosshair_Border
     tiling=NoTile
     imageSize=<8 8>
     margin=<ltrb 0.5 0.5 0.5 0.5>
     outlineSettings {
      bUseBrushTransparency=false
      roundingType=HalfHeightRadius
      width=0
      color {colorUseRule=UseColor_Specified, specifiedColor=<rgba 0 0 0 0>}
      cornerRadii=<0 0 0 0>
     }
     tintColor {colorUseRule=UseColor_Specified, specifiedColor=<rgba 1 1 1 1>}
     uVRegion {bIsValid=false, max=<0 0>, min=<0 0>}
    }}
   slot {class=SizeBoxSlot}
  }
  SizeBox_3 {
   widgetClassPath=&SizeBox
   parent=TopRight
   widget=SizeBox_3
   properties {bOverride_HeightOverride=true, bOverride_WidthOverride=true, heightOverride=8, widthOverride=80}
   slot {class=OverlaySlot}
  }
  6 {
   widgetClassPath=&Image
   bIsVariable=true
   parent=SizeBox_3
   widget=6
   properties {brush {
     bIsDynamicallyLoaded=false
     drawAs=Box
     imageType=NoImage
     mirroring=NoMirror
     resourceName=None
     resourceObject=&M_UI_HomingRocket_Crosshair_Border
     tiling=NoTile
     imageSize=<8 8>
     margin=<ltrb 0.5 0.5 0.5 0.5>
     outlineSettings {
      bUseBrushTransparency=false
      roundingType=HalfHeightRadius
      width=0
      color {colorUseRule=UseColor_Specified, specifiedColor=<rgba 0 0 0 0>}
      cornerRadii=<0 0 0 0>
     }
     tintColor {colorUseRule=UseColor_Specified, specifiedColor=<rgba 1 1 1 1>}
     uVRegion {bIsValid=false, max=<0 0>, min=<0 0>}
    }}
   slot {class=SizeBoxSlot}
  }
  BottomRight {
   widgetClassPath=&Overlay
   parent=SquareBrackets
   widget=BottomRight
   properties {renderTransform {angle=180, scale=<1 1>, shear=<0 0>, translation=<2 2>}}
   slot {class=OverlaySlot, properties {horizontalAlignment=HAlign_Right, verticalAlignment=VAlign_Bottom}}
  }
  SizeBox_6 {
   widgetClassPath=&SizeBox
   parent=BottomRight
   widget=SizeBox_6
   properties {bOverride_HeightOverride=true, bOverride_WidthOverride=true, heightOverride=80, widthOverride=8}
   slot {class=OverlaySlot}
  }
  7 {
   widgetClassPath=&Image
   bIsVariable=true
   parent=SizeBox_6
   widget=7
   properties {brush {
     bIsDynamicallyLoaded=false
     drawAs=Box
     imageType=NoImage
     mirroring=NoMirror
     resourceName=None
     resourceObject=&M_UI_HomingRocket_Crosshair_Border
     tiling=NoTile
     imageSize=<8 8>
     margin=<ltrb 0.5 0.5 0.5 0.5>
     outlineSettings {
      bUseBrushTransparency=false
      roundingType=HalfHeightRadius
      width=0
      color {colorUseRule=UseColor_Specified, specifiedColor=<rgba 0 0 0 0>}
      cornerRadii=<0 0 0 0>
     }
     tintColor {colorUseRule=UseColor_Specified, specifiedColor=<rgba 1 1 1 1>}
     uVRegion {bIsValid=false, max=<0 0>, min=<0 0>}
    }}
   slot {class=SizeBoxSlot}
  }
  SizeBox_7 {
   widgetClassPath=&SizeBox
   parent=BottomRight
   widget=SizeBox_7
   properties {bOverride_HeightOverride=true, bOverride_WidthOverride=true, heightOverride=8, widthOverride=80}
   slot {class=OverlaySlot}
  }
  8 {
   widgetClassPath=&Image
   bIsVariable=true
   parent=SizeBox_7
   widget=8
   properties {brush {
     bIsDynamicallyLoaded=false
     drawAs=Box
     imageType=NoImage
     mirroring=NoMirror
     resourceName=None
     resourceObject=&M_UI_HomingRocket_Crosshair_Border
     tiling=NoTile
     imageSize=<8 8>
     margin=<ltrb 0.5 0.5 0.5 0.5>
     outlineSettings {
      bUseBrushTransparency=false
      roundingType=HalfHeightRadius
      width=0
      color {colorUseRule=UseColor_Specified, specifiedColor=<rgba 0 0 0 0>}
      cornerRadii=<0 0 0 0>
     }
     tintColor {colorUseRule=UseColor_Specified, specifiedColor=<rgba 1 1 1 1>}
     uVRegion {bIsValid=false, max=<0 0>, min=<0 0>}
    }}
   slot {class=SizeBoxSlot}
  }
  AnimBoundTrianglesBorder {
   widgetClassPath=&Border
   bIsVariable=true
   parent=Overlay_100
   widget=AnimBoundTrianglesBorder
   properties {
    background {
     bIsDynamicallyLoaded=false
     drawAs=NoDrawType
     imageType=NoImage
     mirroring=NoMirror
     resourceName=None
     resourceObject=null
     tiling=NoTile
     imageSize=<32 32>
     margin=<ltrb 0 0 0 0>
     outlineSettings {
      bUseBrushTransparency=false
      roundingType=HalfHeightRadius
      width=0
      color {colorUseRule=UseColor_Specified, specifiedColor=<rgba 0 0 0 0>}
      cornerRadii=<0 0 0 0>
     }
     tintColor {colorUseRule=UseColor_Specified, specifiedColor=<rgba 1 1 1 1>}
     uVRegion {bIsValid=false, max=<0 0>, min=<0 0>}
    }
    desiredSizeScale=<1.5 1.5>
    padding=<ltrb 0 0 0 0>
   }
   slot {class=OverlaySlot, properties {horizontalAlignment=HAlign_Center, verticalAlignment=VAlign_Center}}
  }
  TrianglesSizeBox {
   widgetClassPath=&SizeBox
   parent=AnimBoundTrianglesBorder
   widget=TrianglesSizeBox
   properties {bOverride_HeightOverride=true, bOverride_WidthOverride=true, heightOverride=320, widthOverride=320}
   slot {class=BorderSlot, properties {padding=<ltrb 0 0 0 0>}}
  }
  TriangleBrackets {widgetClassPath=&Overlay, parent=TrianglesSizeBox, widget=TriangleBrackets, slot {class=SizeBoxSlot}}
  Triangle_TopLeft_Glow {
   widgetClassPath=&Image
   bIsVariable=true
   parent=TriangleBrackets
   widget=Triangle_TopLeft_Glow
   properties {
    brush {
     bIsDynamicallyLoaded=false
     drawAs=Image
     imageType=NoImage
     mirroring=NoMirror
     resourceName=None
     resourceObject=/ResolveGameplay/Items/Guns/HomingRocket/Widgets/UI_Materials/M_UI_Homing_TriangleBracket_Glow
     tiling=NoTile
     imageSize=<40 40>
     margin=<ltrb 0 0 0 0>
     outlineSettings {
      bUseBrushTransparency=false
      roundingType=HalfHeightRadius
      width=0
      color {colorUseRule=UseColor_Specified, specifiedColor=<rgba 0 0 0 0>}
      cornerRadii=<0 0 0 0>
     }
     tintColor {colorUseRule=UseColor_Specified, specifiedColor=<rgba 1 1 1 1>}
     uVRegion {bIsValid=false, max=<0 0>, min=<0 0>}
    }
    renderTransform {angle=0, scale=<1 1>, shear=<0 0>, translation=<1 2>}
   }
   slot {class=OverlaySlot}
  }
  Triangle_TopLeft {
   widgetClassPath=&Image
   bIsVariable=true
   parent=TriangleBrackets
   widget=Triangle_TopLeft
   properties {brush {
     bIsDynamicallyLoaded=false
     drawAs=Image
     imageType=NoImage
     mirroring=NoMirror
     resourceName=None
     resourceObject=/ResolveGameplay/Items/Guns/HomingRocket/Widgets/UI_Materials/M_UI_Homing_TriangleBracket
     tiling=NoTile
     imageSize=<40 40>
     margin=<ltrb 0 0 0 0>
     outlineSettings {
      bUseBrushTransparency=false
      roundingType=HalfHeightRadius
      width=0
      color {colorUseRule=UseColor_Specified, specifiedColor=<rgba 0 0 0 0>}
      cornerRadii=<0 0 0 0>
     }
     tintColor {colorUseRule=UseColor_Specified, specifiedColor=<rgba 1 1 1 1>}
     uVRegion {bIsValid=false, max=<0 0>, min=<0 0>}
    }}
   slot {class=OverlaySlot}
  }
  Triangle_TopRight_Glow {
   widgetClassPath=&Image
   bIsVariable=true
   parent=TriangleBrackets
   widget=Triangle_TopRight_Glow
   properties {
    brush {
     bIsDynamicallyLoaded=false
     drawAs=Image
     imageType=NoImage
     mirroring=NoMirror
     resourceName=None
     resourceObject=&M_UI_Homing_TriangleBracket_Glow
     tiling=NoTile
     imageSize=<40 40>
     margin=<ltrb 0 0 0 0>
     outlineSettings {
      bUseBrushTransparency=false
      roundingType=HalfHeightRadius
      width=0
      color {colorUseRule=UseColor_Specified, specifiedColor=<rgba 0 0 0 0>}
      cornerRadii=<0 0 0 0>
     }
     tintColor {colorUseRule=UseColor_Specified, specifiedColor=<rgba 1 1 1 1>}
     uVRegion {bIsValid=false, max=<0 0>, min=<0 0>}
    }
    renderTransform {angle=90, scale=<1 1>, shear=<0 0>, translation=<-2 2>}
   }
   slot {class=OverlaySlot, properties {horizontalAlignment=HAlign_Right}}
  }
  Triangle_TopRight {
   widgetClassPath=&Image
   bIsVariable=true
   parent=TriangleBrackets
   widget=Triangle_TopRight
   properties {
    brush {
     bIsDynamicallyLoaded=false
     drawAs=Image
     imageType=NoImage
     mirroring=NoMirror
     resourceName=None
     resourceObject=&M_UI_Homing_TriangleBracket
     tiling=NoTile
     imageSize=<40 40>
     margin=<ltrb 0 0 0 0>
     outlineSettings {
      bUseBrushTransparency=false
      roundingType=HalfHeightRadius
      width=0
      color {colorUseRule=UseColor_Specified, specifiedColor=<rgba 0 0 0 0>}
      cornerRadii=<0 0 0 0>
     }
     tintColor {colorUseRule=UseColor_Specified, specifiedColor=<rgba 1 1 1 1>}
     uVRegion {bIsValid=false, max=<0 0>, min=<0 0>}
    }
    renderTransform {angle=90, scale=<1 1>, shear=<0 0>, translation=<0 0>}
   }
   slot {class=OverlaySlot, properties {horizontalAlignment=HAlign_Right}}
  }
  Triangle_BottomLeft_Glow {
   widgetClassPath=&Image
   bIsVariable=true
   parent=TriangleBrackets
   widget=Triangle_BottomLeft_Glow
   properties {
    brush {
     bIsDynamicallyLoaded=false
     drawAs=Image
     imageType=NoImage
     mirroring=NoMirror
     resourceName=None
     resourceObject=&M_UI_Homing_TriangleBracket_Glow
     tiling=NoTile
     imageSize=<40 40>
     margin=<ltrb 0 0 0 0>
     outlineSettings {
      bUseBrushTransparency=false
      roundingType=HalfHeightRadius
      width=0
      color {colorUseRule=UseColor_Specified, specifiedColor=<rgba 0 0 0 0>}
      cornerRadii=<0 0 0 0>
     }
     tintColor {colorUseRule=UseColor_Specified, specifiedColor=<rgba 1 1 1 1>}
     uVRegion {bIsValid=false, max=<0 0>, min=<0 0>}
    }
    renderTransform {angle=-90, scale=<1 1>, shear=<0 0>, translation=<2 -1>}
   }
   slot {class=OverlaySlot, properties {verticalAlignment=VAlign_Bottom}}
  }
  Triangle_BottomLeft {
   widgetClassPath=&Image
   bIsVariable=true
   parent=TriangleBrackets
   widget=Triangle_BottomLeft
   properties {
    brush {
     bIsDynamicallyLoaded=false
     drawAs=Image
     imageType=NoImage
     mirroring=NoMirror
     resourceName=None
     resourceObject=&M_UI_Homing_TriangleBracket
     tiling=NoTile
     imageSize=<40 40>
     margin=<ltrb 0 0 0 0>
     outlineSettings {
      bUseBrushTransparency=false
      roundingType=HalfHeightRadius
      width=0
      color {colorUseRule=UseColor_Specified, specifiedColor=<rgba 0 0 0 0>}
      cornerRadii=<0 0 0 0>
     }
     tintColor {colorUseRule=UseColor_Specified, specifiedColor=<rgba 1 1 1 1>}
     uVRegion {bIsValid=false, max=<0 0>, min=<0 0>}
    }
    renderTransform {angle=-90, scale=<1 1>, shear=<0 0>, translation=<0 0>}
   }
   slot {class=OverlaySlot, properties {verticalAlignment=VAlign_Bottom}}
  }
  Triangle_BottomRight_Glow {
   widgetClassPath=&Image
   bIsVariable=true
   parent=TriangleBrackets
   widget=Triangle_BottomRight_Glow
   properties {
    brush {
     bIsDynamicallyLoaded=false
     drawAs=Image
     imageType=NoImage
     mirroring=NoMirror
     resourceName=None
     resourceObject=&M_UI_Homing_TriangleBracket_Glow
     tiling=NoTile
     imageSize=<40 40>
     margin=<ltrb 0 0 0 0>
     outlineSettings {
      bUseBrushTransparency=false
      roundingType=HalfHeightRadius
      width=0
      color {colorUseRule=UseColor_Specified, specifiedColor=<rgba 0 0 0 0>}
      cornerRadii=<0 0 0 0>
     }
     tintColor {colorUseRule=UseColor_Specified, specifiedColor=<rgba 1 1 1 1>}
     uVRegion {bIsValid=false, max=<0 0>, min=<0 0>}
    }
    renderTransform {angle=180, scale=<1 1>, shear=<0 0>, translation=<-1 -2>}
   }
   slot {class=OverlaySlot, properties {horizontalAlignment=HAlign_Right, verticalAlignment=VAlign_Bottom}}
  }
  Triangle_BottomRight {
   widgetClassPath=&Image
   bIsVariable=true
   parent=TriangleBrackets
   widget=Triangle_BottomRight
   properties {
    brush {
     bIsDynamicallyLoaded=false
     drawAs=Image
     imageType=NoImage
     mirroring=NoMirror
     resourceName=None
     resourceObject=&M_UI_Homing_TriangleBracket
     tiling=NoTile
     imageSize=<40 40>
     margin=<ltrb 0 0 0 0>
     outlineSettings {
      bUseBrushTransparency=false
      roundingType=HalfHeightRadius
      width=0
      color {colorUseRule=UseColor_Specified, specifiedColor=<rgba 0 0 0 0>}
      cornerRadii=<0 0 0 0>
     }
     tintColor {colorUseRule=UseColor_Specified, specifiedColor=<rgba 1 1 1 1>}
     uVRegion {bIsValid=false, max=<0 0>, min=<0 0>}
    }
    renderTransform {angle=180, scale=<1 1>, shear=<0 0>, translation=<0 0>}
   }
   slot {class=OverlaySlot, properties {horizontalAlignment=HAlign_Right, verticalAlignment=VAlign_Bottom}}
  }
  Status {
   widgetClassPath=&SizeBox
   parent=Overlay_100
   widget=Status
   properties {
    bOverride_HeightOverride=true
    bOverride_WidthOverride=true
    heightOverride=96
    widthOverride=368
    renderTransform {angle=0, scale=<1 1>, shear=<0 0>, translation=<0 44>}
   }
   slot {class=OverlaySlot, properties {horizontalAlignment=HAlign_Center, verticalAlignment=VAlign_Bottom}}
  }
  StatusContainer {
   widgetClassPath=&Border
   bIsVariable=true
   parent=Status
   widget=StatusContainer
   properties {
    background {
     bIsDynamicallyLoaded=false
     drawAs=Box
     imageType=NoImage
     mirroring=NoMirror
     resourceName=None
     resourceObject=/ResolveGameplay/Items/Guns/HomingRocket/Widgets/UI_Materials/M_UI_HomingRocket_Status_Glow
     tiling=NoTile
     imageSize=<64 64>
     margin=<ltrb 0.5 0.5 0.5 0.5>
     outlineSettings {
      bUseBrushTransparency=false
      roundingType=HalfHeightRadius
      width=0
      color {colorUseRule=UseColor_Specified, specifiedColor=<rgba 0 0 0 0>}
      cornerRadii=<0 0 0 0>
     }
     tintColor {colorUseRule=UseColor_Specified, specifiedColor=<rgba 1 1 1 1>}
     uVRegion {bIsValid=false, max=<0 0>, min=<0 0>}
    }
    padding=<ltrb 16 16 16 16>
   }
   slot {class=SizeBoxSlot}
  }
  Overlay_0 {
   widgetClassPath=&Overlay
   parent=StatusContainer
   widget=Overlay_0
   slot {class=BorderSlot, properties {padding=<ltrb 16 16 16 16>}}
  }
  SearchingFX {
   widgetClassPath=&Overlay
   bIsVariable=true
   parent=Overlay_0
   widget=SearchingFX
   slot {class=OverlaySlot, properties {horizontalAlignment=HAlign_Fill, verticalAlignment=VAlign_Fill}}
  }
  GlitchFX1 {
   widgetClassPath=&Image
   bIsVariable=true
   parent=SearchingFX
   widget=GlitchFX1
   properties {
    visibility=SelfHitTestInvisible
    brush {
     bIsDynamicallyLoaded=false
     drawAs=Image
     imageType=NoImage
     mirroring=NoMirror
     resourceName=None
     resourceObject=/ResolveGameplay/Items/Guns/HomingRocket/Widgets/UI_Materials/M_UI_Homing_Pattern
     tiling=NoTile
     imageSize=<32 32>
     margin=<ltrb 0 0 0 0>
     outlineSettings {
      bUseBrushTransparency=false
      roundingType=HalfHeightRadius
      width=0
      color {colorUseRule=UseColor_Specified, specifiedColor=<rgba 0 0 0 0>}
      cornerRadii=<0 0 0 0>
     }
     tintColor {colorUseRule=UseColor_Specified, specifiedColor=<rgba 1 1 1 1>}
     uVRegion {bIsValid=false, max=<0 0>, min=<0 0>}
    }
   }
   slot {class=OverlaySlot, properties {horizontalAlignment=HAlign_Fill, verticalAlignment=VAlign_Fill}}
  }
  GlitchFX2 {
   widgetClassPath=&Image
   bIsVariable=true
   parent=SearchingFX
   widget=GlitchFX2
   properties {
    visibility=SelfHitTestInvisible
    brush {
     bIsDynamicallyLoaded=false
     drawAs=Image
     imageType=NoImage
     mirroring=NoMirror
     resourceName=None
     resourceObject=/ResolveGameplay/Items/Guns/HomingRocket/Widgets/UI_Materials/M_UI_Homing_Pattern_2
     tiling=NoTile
     imageSize=<32 32>
     margin=<ltrb 0 0 0 0>
     outlineSettings {
      bUseBrushTransparency=false
      roundingType=HalfHeightRadius
      width=0
      color {colorUseRule=UseColor_Specified, specifiedColor=<rgba 0 0 0 0>}
      cornerRadii=<0 0 0 0>
     }
     tintColor {colorUseRule=UseColor_Specified, specifiedColor=<rgba 1 1 1 1>}
     uVRegion {bIsValid=false, max=<0 0>, min=<0 0>}
    }
   }
   slot {class=OverlaySlot, properties {horizontalAlignment=HAlign_Fill, verticalAlignment=VAlign_Fill}}
  }
  Progress_Container {
   widgetClassPath=&Image
   bIsVariable=true
   parent=Overlay_0
   widget=Progress_Container
   properties {
    brush {
     bIsDynamicallyLoaded=false
     drawAs=Image
     imageType=NoImage
     mirroring=NoMirror
     resourceName=None
     resourceObject=/ResolveGameplay/Items/Guns/HomingRocket/Widgets/UI_Materials/M_UI_Homing_Progress_Backing
     tiling=NoTile
     imageSize=<32 32>
     margin=<ltrb 0 0 0 0>
     outlineSettings {
      bUseBrushTransparency=false
      roundingType=HalfHeightRadius
      width=0
      color {colorUseRule=UseColor_Specified, specifiedColor=<rgba 0 0 0 0>}
      cornerRadii=<0 0 0 0>
     }
     tintColor {colorUseRule=UseColor_Specified, specifiedColor=<rgba 1 1 1 1>}
     uVRegion {bIsValid=false, max=<0 0>, min=<0 0>}
    }
    colorAndOpacity=<rgba 1 1 1 0.9>
   }
   slot {class=OverlaySlot, properties {horizontalAlignment=HAlign_Fill, verticalAlignment=VAlign_Fill}}
  }
  Progress {
   widgetClassPath=&Image
   bIsVariable=true
   parent=Overlay_0
   widget=Progress
   properties {
    brush {
     bIsDynamicallyLoaded=false
     drawAs=Image
     imageType=NoImage
     mirroring=NoMirror
     resourceName=None
     resourceObject=/ResolveGameplay/Items/Guns/HomingRocket/Widgets/UI_Materials/M_UI_Homing_Progress
     tiling=NoTile
     imageSize=<32 32>
     margin=<ltrb 0 0 0 0>
     outlineSettings {
      bUseBrushTransparency=false
      roundingType=HalfHeightRadius
      width=0
      color {colorUseRule=UseColor_Specified, specifiedColor=<rgba 0 0 0 0>}
      cornerRadii=<0 0 0 0>
     }
     tintColor {colorUseRule=UseColor_Specified, specifiedColor=<rgba 1 1 1 1>}
     uVRegion {bIsValid=false, max=<0 0>, min=<0 0>}
    }
    colorAndOpacity=<rgba 1 1 1 0.9>
   }
   slot {class=OverlaySlot, properties {horizontalAlignment=HAlign_Fill, verticalAlignment=VAlign_Fill, padding=<ltrb 8 8 8 8>}}
  }
  FortHUDElementWrapper_Text {
   widgetClassPath=&FortHUDElementWrapper
   bIsVariable=true
   parent=Overlay_0
   widget=FortHUDElementWrapper_Text
   properties {elementTag {tagName=HUD.Resolve.HomingRocket.LockOnText}}
   slot {class=OverlaySlot, properties {horizontalAlignment=HAlign_Fill, verticalAlignment=VAlign_Fill}}
  }
  Text {widgetClassPath=&WidgetSwitcher, bIsVariable=true, namedSlotHost=FortHUDElementWrapper_Text, widget=Text}
  SearchingText {
   widgetClassPath=&Border
   bIsVariable=true
   parent=Text
   widget=SearchingText
   properties {
    horizontalAlignment=HAlign_Center
    background {
     bIsDynamicallyLoaded=false
     drawAs=Box
     imageType=NoImage
     mirroring=NoMirror
     resourceName=None
     resourceObject=/ResolveGameplay/Items/Guns/HomingRocket/Widgets/UI_Materials/M_UI_HomingRocket_RadialGlow_Searching
     tiling=NoTile
     imageSize=<56 56>
     margin=<ltrb 0.5 0.5 0.5 0.5>
     outlineSettings {
      bUseBrushTransparency=false
      roundingType=HalfHeightRadius
      width=0
      color {colorUseRule=UseColor_Specified, specifiedColor=<rgba 0 0 0 0>}
      cornerRadii=<0 0 0 0>
     }
     tintColor {colorUseRule=UseColor_Specified, specifiedColor=<rgba 1 1 1 1>}
     uVRegion {bIsValid=false, max=<0 0>, min=<0 0>}
    }
    padding=<ltrb 24 16 24 16>
   }
   slot {class=WidgetSwitcherSlot, properties {horizontalAlignment=HAlign_Center, verticalAlignment=VAlign_Center}}
  }
  SizeBox_8 {
   widgetClassPath=&SizeBox
   bIsVariable=true
   parent=SearchingText
   widget=SizeBox_8
   properties {bOverride_MaxDesiredWidth=true, maxDesiredWidth=288, widthOverride=288}
   slot {class=BorderSlot, properties {horizontalAlignment=HAlign_Center, padding=<ltrb 24 16 24 16>}}
  }
  ScaleBox_0 {
   widgetClassPath=&ScaleBox
   parent=SizeBox_8
   widget=ScaleBox_0
   properties {stretch=ScaleToFitX, stretchDirection=DownOnly}
   slot {class=SizeBoxSlot}
  }
  SearchingforVehicle {
   widgetClassPath=&CommonTextBlock
   bIsVariable=true
   parent=ScaleBox_0
   widget=SearchingforVehicle
   properties {
    text=Searching for Vehicles
    textTransformPolicy=ToUpper
    font {
     bForceMonospaced=false
     bMaterialIsStencil=false
     fontMaterial=/ResolveGameplay/Items/Guns/HomingRocket/Widgets/UI_Materials/M_UI_Homing_Font_Searching
     fontObject=&BurbankSmall
     letterSpacing=256
     monospacedWidth=1
     size=13
     skewAmount=0
     typefaceFontName=Black
     outlineSettings {
      bApplyOutlineToDropShadows=false
      bMiteredCorners=false
      bSeparateFillAlpha=false
      outlineMaterial=/Game/UI/Foundation/Materials/M-FontShadow
      outlineSize=2
      outlineColor=<rgba 0.375 1 0.375 1>
     }
    }
    shadowOffset=<0 0>
   }
   slot {class=ScaleBoxSlot}
  }
  LockingOnText {
   widgetClassPath=&Border
   bIsVariable=true
   parent=Text
   widget=LockingOnText
   properties {
    horizontalAlignment=HAlign_Center
    background {
     bIsDynamicallyLoaded=false
     drawAs=Box
     imageType=NoImage
     mirroring=NoMirror
     resourceName=None
     resourceObject=/ResolveGameplay/Items/Guns/HomingRocket/Widgets/UI_Materials/M_UI_HomingRocket_RadialGlow_Locking
     tiling=NoTile
     imageSize=<56 56>
     margin=<ltrb 0.5 0.5 0.5 0.5>
     outlineSettings {
      bUseBrushTransparency=false
      roundingType=HalfHeightRadius
      width=0
      color {colorUseRule=UseColor_Specified, specifiedColor=<rgba 0 0 0 0>}
      cornerRadii=<0 0 0 0>
     }
     tintColor {colorUseRule=UseColor_Specified, specifiedColor=<rgba 1 1 1 1>}
     uVRegion {bIsValid=false, max=<0 0>, min=<0 0>}
    }
    padding=<ltrb 24 16 24 16>
   }
   slot {class=WidgetSwitcherSlot, properties {horizontalAlignment=HAlign_Center, verticalAlignment=VAlign_Center}}
  }
  LockingOn {
   widgetClassPath=&CommonTextBlock
   bIsVariable=true
   parent=LockingOnText
   widget=LockingOn
   properties {
    mobileFontSizeMultiplier=1.2
    text=Locking On
    textTransformPolicy=ToUpper
    font {
     bForceMonospaced=false
     bMaterialIsStencil=false
     fontMaterial=/ResolveGameplay/Items/Guns/HomingRocket/Widgets/UI_Materials/M_UI_Homing_Font_Locking
     fontObject=&BurbankSmall
     letterSpacing=256
     monospacedWidth=1
     size=16
     skewAmount=0
     typefaceFontName=Black
     outlineSettings {
      bApplyOutlineToDropShadows=false
      bMiteredCorners=false
      bSeparateFillAlpha=false
      outlineMaterial=&M-FontShadow
      outlineSize=2
      outlineColor=<rgba 1 0.5 0 0.75>
     }
    }
    shadowOffset=<0 0>
   }
   slot {class=BorderSlot, properties {horizontalAlignment=HAlign_Center, padding=<ltrb 24 16 24 16>}}
  }
  LockedOnCompleteText {
   widgetClassPath=&Border
   bIsVariable=true
   parent=Text
   widget=LockedOnCompleteText
   properties {
    horizontalAlignment=HAlign_Center
    background {
     bIsDynamicallyLoaded=false
     drawAs=Box
     imageType=NoImage
     mirroring=NoMirror
     resourceName=None
     resourceObject=/ResolveGameplay/Items/Guns/HomingRocket/Widgets/UI_Materials/M_UI_HomingRocket_RadialGlow_Locked
     tiling=NoTile
     imageSize=<56 56>
     margin=<ltrb 0.5 0.5 0.5 0.5>
     outlineSettings {
      bUseBrushTransparency=false
      roundingType=HalfHeightRadius
      width=0
      color {colorUseRule=UseColor_Specified, specifiedColor=<rgba 0 0 0 0>}
      cornerRadii=<0 0 0 0>
     }
     tintColor {colorUseRule=UseColor_Specified, specifiedColor=<rgba 1 1 1 1>}
     uVRegion {bIsValid=false, max=<0 0>, min=<0 0>}
    }
    padding=<ltrb 24 16 24 16>
   }
   slot {class=WidgetSwitcherSlot, properties {horizontalAlignment=HAlign_Center, verticalAlignment=VAlign_Center}}
  }
  LockedOnComplete {
   widgetClassPath=&CommonTextBlock
   bIsVariable=true
   parent=LockedOnCompleteText
   widget=LockedOnComplete
   properties {
    mobileFontSizeMultiplier=1.2
    text=Target Locked
    textTransformPolicy=ToUpper
    font {
     bForceMonospaced=false
     bMaterialIsStencil=false
     fontMaterial=null
     fontObject=&BurbankSmall
     letterSpacing=256
     monospacedWidth=1
     size=16
     skewAmount=0
     typefaceFontName=Black
     outlineSettings {
      bApplyOutlineToDropShadows=false
      bMiteredCorners=false
      bSeparateFillAlpha=false
      outlineMaterial=&M-FontShadow
      outlineSize=3
      outlineColor=<rgba 1 0.1667 0 0.75>
     }
    }
    shadowOffset=<0 0>
   }
   slot {class=BorderSlot, properties {horizontalAlignment=HAlign_Center, padding=<ltrb 24 16 24 16>}}
  }
  NoAmmoText {
   widgetClassPath=&Border
   bIsVariable=true
   parent=Text
   widget=NoAmmoText
   properties {
    horizontalAlignment=HAlign_Center
    background {
     bIsDynamicallyLoaded=false
     drawAs=Box
     imageType=NoImage
     mirroring=NoMirror
     resourceName=None
     resourceObject=&M_UI_HomingRocket_RadialGlow_Locked
     tiling=NoTile
     imageSize=<56 56>
     margin=<ltrb 0.5 0.5 0.5 0.5>
     outlineSettings {
      bUseBrushTransparency=false
      roundingType=HalfHeightRadius
      width=0
      color {colorUseRule=UseColor_Specified, specifiedColor=<rgba 0 0 0 0>}
      cornerRadii=<0 0 0 0>
     }
     tintColor {colorUseRule=UseColor_Specified, specifiedColor=<rgba 1 1 1 1>}
     uVRegion {bIsValid=false, max=<0 0>, min=<0 0>}
    }
    padding=<ltrb 24 16 24 16>
   }
   slot {class=WidgetSwitcherSlot, properties {horizontalAlignment=HAlign_Center, verticalAlignment=VAlign_Center}}
  }
  NoAmmo {
   widgetClassPath=&CommonTextBlock
   bIsVariable=true
   parent=NoAmmoText
   widget=NoAmmo
   properties {
    mobileFontSizeMultiplier=1.2
    text=Out of Ammo
    textTransformPolicy=ToUpper
    font {
     bForceMonospaced=false
     bMaterialIsStencil=false
     fontMaterial=null
     fontObject=&BurbankSmall
     letterSpacing=256
     monospacedWidth=1
     size=16
     skewAmount=0
     typefaceFontName=Black
     outlineSettings {
      bApplyOutlineToDropShadows=false
      bMiteredCorners=false
      bSeparateFillAlpha=false
      outlineMaterial=&M-FontShadow
      outlineSize=3
      outlineColor=<rgba 1 0.1667 0 0.75>
     }
    }
    shadowOffset=<0 0>
   }
   slot {class=BorderSlot, properties {horizontalAlignment=HAlign_Center, padding=<ltrb 24 16 24 16>}}
  }
  Pips {
   widgetClassPath=&Overlay
   parent=Overlay_0
   widget=Pips
   slot {class=OverlaySlot, properties {horizontalAlignment=HAlign_Fill, verticalAlignment=VAlign_Fill}}
  }
  Pip_TopLeft {
   widgetClassPath=&Image
   bIsVariable=true
   parent=Pips
   widget=Pip_TopLeft
   properties {
    brush {
     bIsDynamicallyLoaded=false
     drawAs=Image
     imageType=NoImage
     mirroring=NoMirror
     resourceName=None
     resourceObject=/ResolveGameplay/Items/Guns/HomingRocket/Widgets/UI_Materials/M_UI_HomingRocket_Pip_Searching
     tiling=NoTile
     imageSize=<16 16>
     margin=<ltrb 0 0 0 0>
     outlineSettings {
      bUseBrushTransparency=false
      roundingType=HalfHeightRadius
      width=0
      color {colorUseRule=UseColor_Specified, specifiedColor=<rgba 0 0 0 0>}
      cornerRadii=<0 0 0 0>
     }
     tintColor {colorUseRule=UseColor_Specified, specifiedColor=<rgba 1 1 1 1>}
     uVRegion {bIsValid=false, max=<0 0>, min=<0 0>}
    }
    renderTransform {angle=0, scale=<1 1>, shear=<0 0>, translation=<-8 -8>}
   }
   slot {class=OverlaySlot}
  }
  Pip_TopRight {
   widgetClassPath=&Image
   bIsVariable=true
   parent=Pips
   widget=Pip_TopRight
   properties {
    brush {
     bIsDynamicallyLoaded=false
     drawAs=Image
     imageType=NoImage
     mirroring=NoMirror
     resourceName=None
     resourceObject=&M_UI_HomingRocket_Pip_Searching
     tiling=NoTile
     imageSize=<16 16>
     margin=<ltrb 0 0 0 0>
     outlineSettings {
      bUseBrushTransparency=false
      roundingType=HalfHeightRadius
      width=0
      color {colorUseRule=UseColor_Specified, specifiedColor=<rgba 0 0 0 0>}
      cornerRadii=<0 0 0 0>
     }
     tintColor {colorUseRule=UseColor_Specified, specifiedColor=<rgba 1 1 1 1>}
     uVRegion {bIsValid=false, max=<0 0>, min=<0 0>}
    }
    renderTransform {angle=0, scale=<1 1>, shear=<0 0>, translation=<8 -8>}
   }
   slot {class=OverlaySlot, properties {horizontalAlignment=HAlign_Right}}
  }
  Pip_BottomLeft {
   widgetClassPath=&Image
   bIsVariable=true
   parent=Pips
   widget=Pip_BottomLeft
   properties {
    brush {
     bIsDynamicallyLoaded=false
     drawAs=Image
     imageType=NoImage
     mirroring=NoMirror
     resourceName=None
     resourceObject=&M_UI_HomingRocket_Pip_Searching
     tiling=NoTile
     imageSize=<16 16>
     margin=<ltrb 0 0 0 0>
     outlineSettings {
      bUseBrushTransparency=false
      roundingType=HalfHeightRadius
      width=0
      color {colorUseRule=UseColor_Specified, specifiedColor=<rgba 0 0 0 0>}
      cornerRadii=<0 0 0 0>
     }
     tintColor {colorUseRule=UseColor_Specified, specifiedColor=<rgba 1 1 1 1>}
     uVRegion {bIsValid=false, max=<0 0>, min=<0 0>}
    }
    renderTransform {angle=0, scale=<1 1>, shear=<0 0>, translation=<-8 8>}
   }
   slot {class=OverlaySlot, properties {verticalAlignment=VAlign_Bottom}}
  }
  Pip_BottomRight {
   widgetClassPath=&Image
   bIsVariable=true
   parent=Pips
   widget=Pip_BottomRight
   properties {
    brush {
     bIsDynamicallyLoaded=false
     drawAs=Image
     imageType=NoImage
     mirroring=NoMirror
     resourceName=None
     resourceObject=&M_UI_HomingRocket_Pip_Searching
     tiling=NoTile
     imageSize=<16 16>
     margin=<ltrb 0 0 0 0>
     outlineSettings {
      bUseBrushTransparency=false
      roundingType=HalfHeightRadius
      width=0
      color {colorUseRule=UseColor_Specified, specifiedColor=<rgba 0 0 0 0>}
      cornerRadii=<0 0 0 0>
     }
     tintColor {colorUseRule=UseColor_Specified, specifiedColor=<rgba 1 1 1 1>}
     uVRegion {bIsValid=false, max=<0 0>, min=<0 0>}
    }
    renderTransform {angle=0, scale=<1 1>, shear=<0 0>, translation=<8 8>}
   }
   slot {class=OverlaySlot, properties {horizontalAlignment=HAlign_Right, verticalAlignment=VAlign_Bottom}}
  }
  Percent {
   widgetClassPath=&Border
   bIsVariable=true
   parent=Overlay_0
   widget=Percent
   properties {
    visibility=Collapsed
    background {
     bIsDynamicallyLoaded=false
     drawAs=Box
     imageType=NoImage
     mirroring=NoMirror
     resourceName=None
     resourceObject=/ResolveGameplay/Items/Guns/HomingRocket/Widgets/UI_Materials/M_UI_HomingRocket_Glow_ProgressText
     tiling=NoTile
     imageSize=<72 72>
     margin=<ltrb 0.5 0.5 0.5 0.5>
     outlineSettings {
      bUseBrushTransparency=false
      roundingType=HalfHeightRadius
      width=0
      color {colorUseRule=UseColor_Specified, specifiedColor=<rgba 0 0 0 0>}
      cornerRadii=<0 0 0 0>
     }
     tintColor {colorUseRule=UseColor_Specified, specifiedColor=<rgba 1 1 1 1>}
     uVRegion {bIsValid=false, max=<0 0>, min=<0 0>}
    }
    padding=<ltrb 20 8 20 8>
    renderTransform {angle=0, scale=<1 1>, shear=<0 0>, translation=<0 64>}
   }
   slot {class=OverlaySlot, properties {horizontalAlignment=HAlign_Center, verticalAlignment=VAlign_Bottom}}
  }
  CompletedLockonTextSwitcher {
   widgetClassPath=&WidgetSwitcher
   bIsVariable=true
   parent=Percent
   widget=CompletedLockonTextSwitcher
   slot {class=BorderSlot, properties {padding=<ltrb 20 8 20 8>}}
  }
  PercentText {
   widgetClassPath=&CommonTextBlock
   bIsVariable=true
   parent=CompletedLockonTextSwitcher
   widget=PercentText
   properties {
    mobileFontSizeMultiplier=1.2
    text=0%
    font {
     bForceMonospaced=false
     bMaterialIsStencil=false
     fontMaterial=/ResolveGameplay/Items/Guns/HomingRocket/Widgets/UI_Materials/M_UI_Homing_Font_Progress
     fontObject=&BurbankSmall
     letterSpacing=320
     monospacedWidth=1
     size=24
     skewAmount=0
     typefaceFontName=Black
     outlineSettings {
      bApplyOutlineToDropShadows=false
      bMiteredCorners=false
      bSeparateFillAlpha=false
      outlineMaterial=/FortUILibrary/Materials/Foundation/M-FontShadow
      outlineSize=2
      outlineColor=<rgba 1 0.525 0.05 1>
     }
    }
    shadowOffset=<0 0>
   }
   slot {class=WidgetSwitcherSlot}
  }
  PercentText_Completed {
   widgetClassPath=&CommonTextBlock
   bIsVariable=true
   parent=CompletedLockonTextSwitcher
   widget=PercentText_Completed
   properties {
    mobileFontSizeMultiplier=1.2
    text=100%
    font {
     bForceMonospaced=false
     bMaterialIsStencil=false
     fontMaterial=/ResolveGameplay/Items/Guns/HomingRocket/Widgets/UI_Materials/M_UI_Homing_Font_Completed
     fontObject=&BurbankSmall
     letterSpacing=320
     monospacedWidth=1
     size=24
     skewAmount=0
     typefaceFontName=Black
     outlineSettings {
      bApplyOutlineToDropShadows=false
      bMiteredCorners=false
      bSeparateFillAlpha=false
      outlineMaterial=&M-FontShadow
      outlineSize=2
      outlineColor=<rgba 1 1 0.2 1>
     }
    }
    shadowOffset=<0 0>
   }
   slot {class=WidgetSwitcherSlot}
  }
  TutorialHighlight {widgetClassPath=&TutorialHighlight}
 ]
}
