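# Homing rocket HUD warning widget that shows a "Missile Incoming!" banner with glow pulse and sheared stripes inside a material-driven RetainerBox.
# On construct it collapses, binds vehicle enter/exit delegates and listens for WarningStateMessage on the playspace's Gameplay Event Router.
# Messages cache the active flag and refresh visibility, with debug prints.
# Vehicle exit handler collapses it; enter handler re-applies the cached flag, while messages themselves skip any vehicle check.
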WidgetBlueprint /ResolveGameplay/Items/Guns/HomingRocket/Widgets/W_HomingRocket_LockOnWarning
graphs [
 CheckShouldBeVisible {nodes [
   CheckShouldBeVisible() {is=title K2Node_FunctionEntry}
   n1 = Get bCachedActiveState() -> (bCachedActiveState:bool)
   > SetVisibility(InVisibility=Select(Option 0=Collapsed, Option 1=HitTestInvisible, Index=@n1))
   PrintString!(
    execute
    InString=Append(A="Is Warning Active? - ", B="To String (Boolean)"(InBool=@n1))
    bPrintToScreen
    bPrintToLog
    TextColor=(R=0,G=1,B=0.2451,A=1)
    Duration=10
    Key
   ) -> (then)
  ]}
 EventGraph {nodes [
   HandleExitVehicle() {is=title K2Node_CustomEvent}
   > SetVisibility(InVisibility=Collapsed)
   HandleEnterVehicle() {is=title K2Node_CustomEvent}
   > CheckShouldBeVisible() {is=title K2Node_CallFunction}
   Event Construct()
   > SetVisibility(InVisibility=Collapsed)
   n1 = Get FortGlobalUIContext() -> (ReturnValue:Fort Global UIContext Object Reference) {is=title K2Node_GetSubsystem}
   > Bind Event to On Exit Vehicle(self:Fort Global UIContext Object Reference=@n1, Delegate=Create Event()) {is=title K2Node_AddDelegate}
   > Bind Event to On Enter Vehicle Driver(self:Fort Global UIContext Object Reference=@n1, Delegate=Create Event()) {is=title K2Node_AddDelegate}
   > Bind Event to On Enter Vehicle Passenger(self:Fort Global UIContext Object Reference=@n1, Delegate=Create Event()) {is=title K2Node_AddDelegate}
   > n2 = GetEventRouter!(execute, Scope:Object Class Reference=/Script/FortniteGame.FortPlayspace, ContextActor:Actor Object Reference=GetOwningPlayer()) -> (then, ReturnValue:Gameplay Event Router Component Object Reference)
   > n3 = StartListeningToEvent!(
    execute:Exec
    Target:Gameplay Event Router Component Object Reference=@n2
    EventType:Script Struct Object Reference=&/WarningStateMessage
    Context:Object Reference
    EventBubblingRule:EEventBubblingRule Enum
   ) -> (
    then:Exec
    AsyncTaskProxy:Async Action Start Listening To Event Object Reference
    OnEventReceived:Exec
    Payload:Warning State Message Structure
   )
   n4 = Set bCachedActiveState(
    execute=@n3.OnEventReceived
    bCachedActiveState:bool=Break Warning State Message!(WarningStateMessage:"Warning State Message Structure (by ref)"=@n3.Payload) -> (bIsActive_4_#1:bool=~) {is=title K2Node_BreakStruct}
   ) -> (Output_Get:bool)
   > CheckShouldBeVisible()
   PrintString(
    InString=Append(A="Warning message received should be active? - ", B="To String (Boolean)"(InBool=@n4))
    TextColor=(R=0,G=0.1305,B=1,A=1)
    Duration=10
   )
   Event Destruct()
  ]}
]
variables [bCachedActiveState {type=boolean}]
widgets {
 info {inheritedWidgetCount=0, parentClass=&UserWidget, rootWidgetClass=&Overlay, widgetCount=6}
 widgets [
  Overlay_2 {widgetClassPath=&Overlay, widget=Overlay_2}
  Image_GlowPulse {
   widgetClassPath=&Image
   bIsVariable=true
   parent=Overlay_2
   widget=Image_GlowPulse
   properties {
    brush {
     bIsDynamicallyLoaded=false
     drawAs=Image
     imageType=NoImage
     mirroring=NoMirror
     resourceName=None
     resourceObject=/Game/Athena/HUD/LTM/Wax/Materials/M_UI_TextGlow_Pulse
     tiling=NoTile
     imageSize=<32 32>
     margin=<ltrb 0 0 0 0>
     outlineSettings {
      bUseBrushTransparency=false
      roundingType=HalfHeightRadius
      width=0
      color {colorUseRule=UseColor_Specified, specifiedColor=<rgba 0 0 0 0>}
      cornerRadii=<0 0 0 0>
     }
     tintColor {colorUseRule=UseColor_Specified, specifiedColor=<rgba 1 1 1 1>}
     uVRegion {bIsValid=false, max=<0 0>, min=<0 0>}
    }
    colorAndOpacity=<rgba 0.8469 0.1384 0.0168 1>
   }
   slot {class=OverlaySlot, properties {horizontalAlignment=HAlign_Fill, verticalAlignment=VAlign_Fill, padding=<ltrb -200 -100 -200 -100>}}
  }
  RetainerBox_AdditiveContainer {
   widgetClassPath=/Script/UMG.RetainerBox
   bIsVariable=true
   parent=Overlay_2
   widget=RetainerBox_AdditiveContainer
   properties {effectMaterial=/Game/Athena/HUD/Materials/M_UI_SelfDestructRetainer_Inst}
   slot {class=OverlaySlot}
  }
  Overlay_1 {widgetClassPath=&Overlay, parent=RetainerBox_AdditiveContainer, widget=Overlay_1, slot {class=PanelSlot}}
  Image_WarningStripes {
   widgetClassPath=&Image
   bIsVariable=true
   parent=Overlay_1
   widget=Image_WarningStripes
   properties {
    brush {
     bIsDynamicallyLoaded=false
     drawAs=Image
     imageType=NoImage
     mirroring=NoMirror
     resourceName=None
     resourceObject=/Game/Athena/HUD/Materials/M_UI_StripedWarningMaterial_Inst
     tiling=NoTile
     imageSize=<32 32>
     margin=<ltrb 0 0 0 0>
     outlineSettings {
      bUseBrushTransparency=false
      roundingType=HalfHeightRadius
      width=0
      color {colorUseRule=UseColor_Specified, specifiedColor=<rgba 0 0 0 0>}
      cornerRadii=<0 0 0 0>
     }
     tintColor {colorUseRule=UseColor_Specified, specifiedColor=<rgba 1 1 1 1>}
     uVRegion {bIsValid=false, max=<0 0>, min=<0 0>}
    }
    renderTransform {angle=0, scale=<1 1>, shear=<-38 0>, translation=<0 0>}
   }
   slot {class=OverlaySlot, properties {horizontalAlignment=HAlign_Fill, verticalAlignment=VAlign_Fill}}
  }
  CommonTextBlock_WarningText {
   widgetClassPath=&CommonTextBlock
   bIsVariable=true
   parent=Overlay_1
   widget=CommonTextBlock_WarningText
   properties {
    justification=Center
    text="Missile Incoming!"
    textTransformPolicy=ToUpper
    colorAndOpacity {colorUseRule=UseColor_Specified, specifiedColor=<rgba 1 0.6445 0 1>}
    font {
     bForceMonospaced=false
     bMaterialIsStencil=false
     fontMaterial=null
     fontObject=&Burbank
     letterSpacing=0
     monospacedWidth=1
     size=36
     skewAmount=0
     typefaceFontName=BurbankRegular-Black
     outlineSettings {
      bApplyOutlineToDropShadows=false
      bMiteredCorners=false
      bSeparateFillAlpha=false
      outlineMaterial=null
      outlineSize=4
      outlineColor=<rgba 0 0 0 1>
     }
    }
    renderTransform {angle=0, scale=<1 1>, shear=<-12 0>, translation=<0 0>}
    shadowOffset=<0 0>
   }
   slot {class=OverlaySlot, properties {padding=<ltrb 200 15 200 15>}}
  }
 ]
}
